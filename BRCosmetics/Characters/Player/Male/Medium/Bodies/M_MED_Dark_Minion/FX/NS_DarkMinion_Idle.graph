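# Idle Niagara effect for the Dark Minion player body cosmetic, keeping all four emitters attached to the character's skeletal mesh.
# Decal and Glow spawn soft fading sprites continuously, while Smoke bursts SubUV puffs whose curve-driven offsets drift them along negative Z.
# LightBurst loops one particle with flickering sprite, light and component renderers.
# User.Color_02 tints everything except the hardcoded magenta Glow; Color_01 is unused and the front-end preview switch selects identical values.
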NiagaraSystem /BRCosmetics/Characters/Player/Male/Medium/Bodies/M_MED_Dark_Minion/FX/NS_DarkMinion_Idle
dependencies {
 usedDataInterfaces [&NiagaraDataInterfaceCurve, &NiagaraDataInterfaceColorCurve]
 usedDynamicInputs []
 usedModules [&SystemState]
}
emitters {
 Decal {
  emitterUpdateScript [
   &EmitterState()
   &SpawnRate(Use Spawn Probability=false)
  ]
  particleSpawnScript [
   &FN_SetSpawnGroup_v2(Number of Spawn Groups=2)
   /Niagara/Modules/Spawn/Initialization/InitializeParticle(
    Color=$User.Color_02
    Lifetime=1.5
    Module.Write Color=true
    Module.Write Lifetime=true
    Module.Write Mass=true
    Module.Write Material Random=true
    Module.Write Position=false
    Module.Write Scale=false
    Module.Write SpriteRotation=false
    Module.Write SpriteSize=true
    Position=&SimulationPosition()
    Sprite Size=&Vector2DFromFloat(Value=&RandomRangeFloat(Minimum=70, Maximum=70))
   )
   &DynamicMaterialParameters(
    Index 0 Param 1=0.5
    Index 0 Param 2=1
    Index 0 Param 3=1
    Module.Param0WriteEnabled=false
    Module.Param1WriteEnabled=true
    Module.Param2WriteEnabled=true
    Module.Param3WriteEnabled=false
   )
   &BoxLocation(Box Offset=<10 0 -5>, Box Size=<5 10 15>, MaskSpawnGroup=false, Module.Override Seed=false, Rotate Box=false, Surface Only=false)
  ]
  particleUpdateScript [
   &ParticleState()
   &SkeletalMeshLocation(Bone Sampling Mode=::NewEnumerator4, Inherit Velocity=false, MaskSpawnGroup=false, Skeletal Mesh=$User.SkeletalMesh)
   &DynamicMaterialParameters(
    Index 0 Param 1=&FloatFromCurve(FloatCurve=DataInterface() {propertyValues {Curve {keys [<at 0.4 3>,<at 1 6>]}}})
    Module.Param0WriteEnabled=true
    Module.Param1WriteEnabled=false
    Module.Param2WriteEnabled=false
    Module.Param3WriteEnabled=false
   ) {module=DynamicMaterialParameters001}
   &ScaleColor(
    Scale Alpha=&FloatFromCurve(FloatCurve=DataInterface() {propertyValues {Curve {keys [<at 0 0 RCIM_Cubic>,<at 0.2 1 RCIM_Cubic>,<at 1 0 RCIM_Cubic>]}}}, Scale Curve=0.25)
    ScaleA=true
    ScaleRGB=true
   )
   &CameraOffset(Camera Offset Amount=&RandomRangeFloat(Minimum=5, Maximum=10))
  ]
  rendererClasses [&NiagaraSpriteRendererProperties]
  renderers [{rendererClass=&NiagaraSpriteRendererProperties}]
 }
 Glow {
  emitterUpdateScript [
   &EmitterState()
   &SpawnRate(SpawnRate=1.5, Use Spawn Probability=false)
  ]
  particleSpawnScript [
   &InitializeParticle(
    Color=<rgba 5 5 5 0.05>
    Lifetime=&UniformRangedFloat(Minimum=2, Maximum=2.5)
    Module.Write Color=true
    Module.Write Lifetime=true
    Module.Write Mass=false
    Module.Write Material Random=true
    Module.Write Position=false
    Module.Write Scale=false
    Module.Write SpriteRotation=false
    Module.Write SpriteSize=true
    Position=&SimulationPosition()
    Sprite Rotation=&UniformRangedFloat(Minimum=-180, Maximum=180)
    Sprite Size=&Vector2DFromFloat(Value=&UniformRangedFloat(Minimum=30, Maximum=50))
   )
   &CameraOffset(Camera Offset Amount=&RandomRangeFloat(Minimum=-5, Maximum=10))
   &AddVelocity(Velocity=&RandomRangeVector(Minimum=<-1 -1 5>, Maximum=<1 1 2>))
  ]
  particleUpdateScript [
   &SkeletalMeshLocation(
    Bone Sampling Mode=::NewEnumerator4
    Inherit Velocity=false
    MaskSpawnGroup=false
    Sampled Position Offset=&RandomRangeVector(Minimum=<-2 -2 -2>, Maximum=<2 2 2>)
    Skeletal Mesh=$User.SkeletalMesh
   )
   &SolveForcesAndVelocity(Clamp Velocity=false, Limit Acceleration=false)
   &ParticleState()
   &ScaleSpriteSize(Scale Factor=&Vector2DFromFloat(Value=&FloatFromCurve(FloatCurve=DataInterface() {propertyValues {Curve {keys [
          {arriveTangent=7.5764, interpMode=RCIM_Cubic, leaveTangent=7.5765, tangentMode=RCTM_User, time=0, value=0.02}
          {arriveTangent=1.2063, interpMode=RCIM_Cubic, leaveTangent=1.2063, tangentMode=RCTM_User, time=0.18, value=0.99}
          <at 1 0.5 RCIM_Cubic>
         ]}}})))
   &ScaleColor(
    Scale Mode=::NewEnumerator1
    Scale RGBA=&MakeVector4_FromLinearColor(LinearColor=&ColorFromCurve(ColorCurve=DataInterface() {propertyValues {
        AlphaCurve {keys [<at 0 0 RCIM_Cubic>,<at 0.1765 0.5>,<at 1.0484 0 RCIM_Cubic>]}
        BlueCurve {keys [<at 0.247 1 RCIM_Cubic>]}
        GreenCurve {keys [<at 0.247 0.2016 RCIM_Cubic>]}
        RedCurve {keys [<at 0.247 1 RCIM_Cubic>]}
       }}))
    ScaleRGBA=true
   )
  ]
  rendererClasses [&NiagaraSpriteRendererProperties]
  renderers [{rendererClass=&NiagaraSpriteRendererProperties}]
 }
 LightBurst {
  emitterUpdateScript [
   &EmitterState(Inactive Response=::NewEnumerator1, Life Cycle Mode=::NewEnumerator1, Loop Duration=4, UseLoopDelay=true)
   &SpawnBurst_Instantaneous(Spawn Count=1, Use Spawn Probability=false)
  ]
  particleSpawnScript [
   &InitializeParticle(
    Lifetime=4
    Module.Write Color=true
    Module.Write Lifetime=true
    Module.Write Mass=false
    Module.Write Material Random=true
    Module.Write Position=false
    Module.Write Scale=false
    Module.Write SpriteRotation=false
    Module.Write SpriteSize=true
    Sprite Size=<80 80>
   )
   &SkeletalMeshLocation(Bone Sampling Mode=::NewEnumerator4, Inherit Velocity=true, MaskSpawnGroup=false, Skeletal Mesh=$User.SkeletalMesh, VelocitySpeedLimit=false)
   &Color(Color=$User.Color_02)
  ]
  particleUpdateScript [
   &ParticleState()
   &SkeletalMeshLocation(
    Bone Sampling Mode=::NewEnumerator4
    Inherit Velocity=false
    MaskSpawnGroup=false
    Sampled Position Offset=<0 10 0>
    Sampled Position Offset Coordinate Space:ENiagaraMeshTransforms=::NewEnumerator3
    Skeletal Mesh=$User.SkeletalMesh
   ) {module=SkeletalMeshLocation001}
   &ScaleColor(
    Scale Alpha=&FloatFromCurve(
     FloatCurve=DataInterface() {propertyValues {Curve {keys [
         <at 0 0 RCIM_Cubic>
         {arriveTangent=3.2234, interpMode=RCIM_Cubic, leaveTangent=3.2234, time=0.1149, value=0.4468}
         {arriveTangent=0.2662, interpMode=RCIM_Cubic, leaveTangent=0.2662, time=0.2325, value=0.7493}
         {arriveTangent=-0.7483, interpMode=RCIM_Cubic, leaveTangent=-0.7483, time=0.3216, value=0.5018}
         {arriveTangent=-1.297, interpMode=RCIM_Cubic, leaveTangent=-1.297, time=0.3917, value=0.6302}
         {arriveTangent=0.5398, interpMode=RCIM_Cubic, leaveTangent=0.5398, time=0.5036, value=0.2658}
         {arriveTangent=0.6847, interpMode=RCIM_Cubic, leaveTangent=0.6847, time=0.6396, value=0.764}
         {arriveTangent=-1.0744, interpMode=RCIM_Cubic, leaveTangent=-1.0744, time=0.7178, value=0.4125}
         {arriveTangent=-1.4697, interpMode=RCIM_Cubic, leaveTangent=-1.4697, time=0.841, value=0.5477}
         <at 0.9984 0 RCIM_Cubic>
        ]}}}
     Scale Curve=0.05
    )
    ScaleA=true
    ScaleRGB=true
   )
   &FN_Light_Attributes(
    Color=$User.Color_02
    Create Color Attribute=true
    Create Exponent Attribute=true
    Create Position Attribute=true
    Create Radius Attribute=true
    Enable Particle Light Enable=true
    Exponent=12
    Light Enabled=true
    Position:NiagaraPosition=$Particles.Position
    Radius=100
    Scale Color=&FloatFromCurve(FloatCurve=DataInterface() {propertyValues {Curve {keys [
         <at 0 0 RCIM_Cubic>
         {arriveTangent=3.2234, interpMode=RCIM_Cubic, leaveTangent=3.2234, time=0.1149, value=0.4468}
         {arriveTangent=0.2662, interpMode=RCIM_Cubic, leaveTangent=0.2662, time=0.2325, value=0.7493}
         {arriveTangent=-0.7483, interpMode=RCIM_Cubic, leaveTangent=-0.7483, time=0.3216, value=0.5018}
         {arriveTangent=-1.297, interpMode=RCIM_Cubic, leaveTangent=-1.297, time=0.3917, value=0.6302}
         {arriveTangent=0.5398, interpMode=RCIM_Cubic, leaveTangent=0.5398, time=0.5036, value=0.2658}
         {arriveTangent=0.6847, interpMode=RCIM_Cubic, leaveTangent=0.6847, time=0.6396, value=0.764}
         {arriveTangent=-1.0744, interpMode=RCIM_Cubic, leaveTangent=-1.0744, time=0.7178, value=0.4125}
         {arriveTangent=-1.4697, interpMode=RCIM_Cubic, leaveTangent=-1.4697, time=0.841, value=0.5477}
         <at 0.9984 0 RCIM_Cubic>
        ]}}})
   )
   &FN_Gen5_Light(
    Attenutation=53
    Color:LinearColor=$User.Color_02
    Exponent=8
    Intensity:NiagaraFloat=&FloatFromCurve(
     FloatCurve=DataInterface() {propertyValues {Curve {keys [
         <at 0 0.5 RCIM_Cubic>
         {arriveTangent=1.0726, interpMode=RCIM_Cubic, leaveTangent=1.0726, time=0.1149, value=0.4468}
         {arriveTangent=0.6624, interpMode=RCIM_Cubic, leaveTangent=0.6624, time=0.2325, value=0.7493}
         {arriveTangent=-1.7837, interpMode=RCIM_Cubic, leaveTangent=-1.7837, time=0.3917, value=0.6302}
         {arriveTangent=0.5398, interpMode=RCIM_Cubic, leaveTangent=0.5398, time=0.5036, value=0.2658}
         {arriveTangent=0.8353, interpMode=RCIM_Cubic, leaveTangent=0.8353, time=0.6396, value=0.764}
         {arriveTangent=-0.7359, interpMode=RCIM_Cubic, leaveTangent=-0.7359, time=0.841, value=0.5477}
         <at 0.9984 0.5 RCIM_Cubic>
        ]}}}
     Scale Curve=50
    )
   )
  ]
  rendererClasses [
   &NiagaraLightRendererProperties
   &NiagaraSpriteRendererProperties
   &NiagaraComponentRendererProperties
  ]
  renderers [
   {rendererClass=&NiagaraLightRendererProperties, rendererIndex=0}
   {rendererClass=&NiagaraSpriteRendererProperties, rendererIndex=1}
   {rendererClass=&NiagaraComponentRendererProperties, rendererIndex=2}
  ]
 }
 Smoke {
  emitterUpdateScript [
   &EmitterState(
    Inactive Response=::NewEnumerator1
    Life Cycle Mode=::NewEnumerator1
    Loop Duration=&RandomRangeFloat(Minimum=0.35, Maximum=0.75, Evaluation Type=::NewEnumerator1)
    Recalculate Duration Each Loop=true
    UseLoopDelay=true
   )
   &SpawnBurst_Instantaneous(Spawn Count=&RandomRangeInt(Minimum=4, Maximum=6), Use Spawn Probability=false)
  ]
  particleSpawnScript [
   /SetVariables(Particles.SpriteRotationRate=&RandomRangeFloat(Minimum=-0.25, Maximum=0.25, Evaluation Type=::NewEnumerator1)) {isSetParameters=true}
   &InitializeParticle(
    Lifetime=&RandomRangeFloat(Minimum=5, Maximum=6, Evaluation Type=::NewEnumerator1)
    Module.Write Color=true
    Module.Write Lifetime=true
    Module.Write Mass=true
    Module.Write Material Random=true
    Module.Write Position=false
    Module.Write Scale=false
    Module.Write SpriteRotation=true
    Module.Write SpriteSize=true
    Sprite Rotation=&NormalizedAngleToDegrees(NormalizedAngle=&RandomRangeFloat(Minimum=0.2, Maximum=1, Evaluation Type=::NewEnumerator1))
    Sprite Size=&Vector2DFromFloat(Value=&RandomRangeFloat(Minimum=35, Maximum=40, Evaluation Type=::NewEnumerator1))
   )
   &Color(Color=$User.Color_02)
   /Niagara/Modules/Spawn/Location/SphereLocation(MaskSpawnGroup=false, Module.Override Randomness=false, Sphere Radius=10, Surface Only=true)
   &DynamicMaterialParameters(
    Index 0 Param 1=&RandomRangeFloat(Minimum=0, Maximum=1, Evaluation Type=::NewEnumerator1)
    Index 0 Param 2=&RandomRangeFloat(Minimum=2, Maximum=4, Evaluation Type=::NewEnumerator1)
    Module.Param0WriteEnabled=true
    Module.Param1WriteEnabled=true
    Module.Param2WriteEnabled=false
    Module.Param3WriteEnabled=false
   )
   &SkeletalMeshLocation(
    Apply Sampled Orientation to Mesh Particles=true
    Inherit Velocity=true
    "Inherit Velocity (Scale)"=0.25
    Inherited Velocity Speed Limit=15
    MaskSpawnGroup=false
    Mesh Sampling Type=::NewEnumerator5
    Sampled Position Offset=&RandomRangeVector(Minimum=<-5 -5 0>, Maximum=<5 5 8>)
    Sampled Position Offset Coordinate Space:ENiagaraMeshTransforms=::NewEnumerator2
    Skeletal Mesh=$User.SkeletalMesh
    Socket Sampling Mode=::NewEnumerator5
    VelocitySpeedLimit=true
   )
  ]
  particleUpdateScript [
   &ParticleState()
   &ScaleColor(
    Scale Alpha=&FloatFromCurve(
     FloatCurve=DataInterface() {propertyValues {Curve {keys [
         <at 0 0 RCIM_Cubic>
         {arriveTangent=4.3234, interpMode=RCIM_Cubic, leaveTangent=4.3234, tangentMode=RCTM_User, time=0.1965, value=0.754}
         {arriveTangent=-0.0267, interpMode=RCIM_Cubic, leaveTangent=-0.0267, time=0.3, value=1}
         {arriveTangent=-1.8385, interpMode=RCIM_Cubic, leaveTangent=-1.8385, tangentMode=RCTM_User, time=0.7626, value=0.7389}
         <at 1 0 RCIM_Cubic>
        ]}}}
     Scale Curve=&MakeCustomFloatFromBool(Bool=$User.bIsFrontEndPreview, True Float=0.75, False Float=0.75)
    )
    Scale RGB=&VectorFromFloat(Value=&FloatFromCurve(FloatCurve=DataInterface() {propertyValues {Curve {keys [<at 0 0.5 RCIM_Cubic>,<at 0.2 1 RCIM_Cubic>,<at 1 0.5 RCIM_Cubic>]}}}))
    ScaleA=true
    ScaleRGB=true
   )
   /Niagara/Modules/Update/SubUV/SubUVAnimation(Number Of Frames=62)
   &DynamicMaterialParameters(Module.Param0WriteEnabled=false, Module.Param1WriteEnabled=false, Module.Param2WriteEnabled=true, Module.Param3WriteEnabled=true) {module=DynamicMaterialParameters001}
   &ScaleSpriteSize(Scale Factor=&Vector2DFromFloat(Value=&FloatFromCurve(FloatCurve=DataInterface() {propertyValues {Curve {keys [{arriveTangent=0.7932, interpMode=RCIM_Cubic, time=0, value=0.5}, {interpMode=RCIM_Cubic, leaveTangent=0.3663, time=1.0038, value=1}]}}})))
   &SpriteRotationRate(Rotation Rate=&NormalizedAngleToDegrees(NormalizedAngle=$Particles.SpriteRotationRate))
   &SkeletalMeshLocation(
    Inherit Velocity=false
    "Inherit Velocity (Scale)"=0.25
    MaskSpawnGroup=false
    Mesh Sampling Type=::NewEnumerator5
    Sampled Position Offset=&Add_Vector(
     A=&RandomRangeVector(Minimum=<-5 -8 0>, Maximum=<5 8 20>)
     B=&MakeVector(
      X=&Multiply_Float(
       A=&RandomRangeFloat(Minimum=-1, Maximum=1)
       B=&FloatFromCurve(
        FloatCurve=DataInterface() {propertyValues {Curve {keys [
            <at 0 0 RCIM_Cubic>
            {arriveTangent=0.4457, interpMode=RCIM_Cubic, leaveTangent=0.4457, time=0.3197, value=0.1025}
            {arriveTangent=0.8904, interpMode=RCIM_Cubic, leaveTangent=0.8904, time=0.6772, value=0.3018}
            {arriveTangent=3.7118, interpMode=RCIM_Cubic, leaveTangent=3.7118, time=0.9138, value=0.6315}
            <at 1 1.5 RCIM_Cubic>
           ]}}}
        Scale Curve=3
       )
      )
      Y=&Multiply_Float(
       A=&RandomRangeFloat(Minimum=-1, Maximum=1)
       B=&FloatFromCurve(
        FloatCurve=DataInterface() {propertyValues {Curve {keys [
            <at 0 0 RCIM_Cubic>
            {arriveTangent=0.4457, interpMode=RCIM_Cubic, leaveTangent=0.4457, time=0.3197, value=0.1025}
            {arriveTangent=0.8904, interpMode=RCIM_Cubic, leaveTangent=0.8904, time=0.6772, value=0.3018}
            {arriveTangent=3.7118, interpMode=RCIM_Cubic, leaveTangent=3.7118, time=0.9138, value=0.6315}
            <at 1 1.5 RCIM_Cubic>
           ]}}}
        Scale Curve=3
       )
      )
      Z=&FloatFromCurve(
       FloatCurve=DataInterface() {propertyValues {Curve {keys [
           <at 0 0 RCIM_Cubic>
           {arriveTangent=0.4457, interpMode=RCIM_Cubic, leaveTangent=0.4457, time=0.3197, value=0.1025}
           {arriveTangent=1.0802, interpMode=RCIM_Cubic, leaveTangent=1.0802, time=0.6772, value=0.3018}
           {arriveTangent=3.7118, interpMode=RCIM_Cubic, leaveTangent=3.7118, time=0.891, value=0.7196}
           <at 1 1.5 RCIM_Cubic>
          ]}}}
       Scale Curve=-120
      )
     )
    )
    Sampled Position Offset Coordinate Space:ENiagaraMeshTransforms=::NewEnumerator2
    Skeletal Mesh=$User.SkeletalMesh
    Socket Sampling Mode=::NewEnumerator5
   ) {module=SkeletalMeshLocation002}
  ]
  rendererClasses [&NiagaraSpriteRendererProperties]
  renderers [{rendererClass=&NiagaraSpriteRendererProperties}]
 }
}
summary {
 systemName=NS_DarkMinion_Idle
 emitters [
  {emitterName=Smoke, rendererClasses [&NiagaraSpriteRendererProperties]}
  {
   emitterName=LightBurst
   rendererClasses [
    &NiagaraLightRendererProperties
    &NiagaraSpriteRendererProperties
    &NiagaraComponentRendererProperties
   ]
  }
  {emitterName=Decal, rendererClasses [&NiagaraSpriteRendererProperties]}
  {emitterName=Glow, rendererClasses [&NiagaraSpriteRendererProperties]}
 ]
 userVariables [
  User.bIsFrontEndPreview {type=NiagaraBool}
  User.Color_01 {type=LinearColor, description="", defaultValue {struct=LinearColor, value=<rgba 0.2462 0.0437 0.5647 0.9216>}}
  User.Color_02 {type=LinearColor, description="", defaultValue {struct=LinearColor, value=<rgba 1 0.2016 1 1>}}
  User.SkeletalMesh {
   type=NiagaraDataInterfaceSkeletalMesh
   defaultValue {
    struct=NiagaraExt_VariableValue_DataInterface
    value {dataInterface=&self:NiagaraDataInterfaceSkeletalMesh_0}
   }
  }
 ]
}
