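# Physics asset for the back skirt of the Dark Minion male medium player body cosmetic.
# Five identical skirt capsules form a serial chain from a tiny driver capsule attached to the pelvis, with swing limits and tight twist limits on each link.
# Pelvis and both wolf-thigh capsules supply body volumes around the skirt.
# Constraints linking the driver and thighs to the pelvis leave all angular axes free.
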PhysicsAsset /BRCosmetics/Characters/Player/Male/Medium/Bodies/M_MED_Dark_Minion/Meshes/M_MED_Dark_Minion_Bck_Skirt
bodies {
 dyn_driver_skirt_bck [{shapeType=Capsule, length=0.505, radius=0.505, rotation=<rot 0 -90 0>}]
 dyn_skirt_bck_01 [{shapeType=Capsule, length=9.25, radius=8.25, center=<6 7.5 0>, rotation=<rot 90 0 0>}]
 dyn_skirt_bck_02 [{shapeType=Capsule, length=9.25, radius=8.25, center=<6 7.5 0>, rotation=<rot 90 0 0>}]
 dyn_skirt_bck_03 [{shapeType=Capsule, length=9.25, radius=8.25, center=<6 7.5 0>, rotation=<rot 90 0 0>}]
 dyn_skirt_bck_04 [{shapeType=Capsule, length=9.25, radius=8.25, center=<6 7.5 0>, rotation=<rot 90 0 0>}]
 dyn_skirt_bck_05 [{shapeType=Capsule, length=9.25, radius=8.25, center=<6 7.5 0>, rotation=<rot 90 0 0>}]
 pelvis [{shapeType=Capsule, length=7.5, radius=10.5, center=<-6.1772 2.8766 0>}]
 wolf_thigh_l [{shapeType=Capsule, length=28.1139, radius=9.2826, center=<-14.3533 4.3761 1.7344>, rotation=<rot 0 -100 -90>}]
 wolf_thigh_r [{shapeType=Capsule, length=28.1139, radius=9.2826, center=<14.3533 -4.3761 -1.7344>, rotation=<rot 0 -100 -90>}]
}
constraints [
 {
  bone1Name=wolf_thigh_l
  bone2Name=pelvis
  swing1LimitDegrees=45
  swing1Motion=Free
  swing2LimitDegrees=45
  swing2Motion=Free
  twistLimitDegrees=45
  twistMotion=Free
 }
 {
  bone1Name=wolf_thigh_r
  bone2Name=pelvis
  swing1LimitDegrees=45
  swing1Motion=Free
  swing2LimitDegrees=45
  swing2Motion=Free
  twistLimitDegrees=45
  twistMotion=Free
 }
 {
  bone1Name=dyn_driver_skirt_bck
  bone2Name=pelvis
  swing1LimitDegrees=45
  swing1Motion=Free
  swing2LimitDegrees=45
  swing2Motion=Free
  twistLimitDegrees=45
  twistMotion=Free
 }
 {bone1Name=dyn_skirt_bck_01, bone2Name=dyn_driver_skirt_bck, swing1LimitDegrees=45, swing2LimitDegrees=45, twistLimitDegrees=5}
 {bone1Name=dyn_skirt_bck_02, bone2Name=dyn_skirt_bck_01, swing1LimitDegrees=45, swing2LimitDegrees=45, twistLimitDegrees=5}
 {bone1Name=dyn_skirt_bck_03, bone2Name=dyn_skirt_bck_02, swing1LimitDegrees=45, swing2LimitDegrees=45, twistLimitDegrees=5}
 {bone1Name=dyn_skirt_bck_04, bone2Name=dyn_skirt_bck_03, swing1LimitDegrees=45, swing2LimitDegrees=45, twistLimitDegrees=5}
 {bone1Name=dyn_skirt_bck_05, bone2Name=dyn_skirt_bck_04, swing1LimitDegrees=45, swing2LimitDegrees=45, twistLimitDegrees=5}
]
