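# Control Rig for the Dark Minion body that drives its extra alt-finger, wolf-leg and arm helper bones from the standard skeleton.
# Alt fingers take projected rotations from their standard counterparts; wolf legs copy thigh, calf and foot rotations, then FABRIK-solve to a foot-relative target.
# Interpolated crouch_alpha adjusts foot offsets, while aim units orient calf, ball, rope, band and upper-arm helpers.
# An older bone-based right-leg chain remains disconnected.
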ControlRigBlueprint /BRCosmetics/Characters/Player/Male/Medium/Bodies/M_MED_Dark_Minion/Meshes/M_MED_Dark_Minion_CtrlRig
graphs [
 RigVMFunctionLibrary {links [], nodes []}
 RigVMModel {nodes [
   n1 = &RigUnit_BeginExecution()
   RigUnit_CollectionItems = /Script/ControlRig.RigUnit_CollectionItems(Items=((Type=Bone,Name="thumb_alt_01_l"),(Type=Bone,Name="ring_alt_01_l"),(Type=Bone,Name="index_alt_01_l"))) {is=struct RigVMUnitNode}
   n2 = /Script/ControlRig.RigUnit_CollectionLoop(
    ExecutePin=@n1.ExecutePin
    Collection=/Script/ControlRig.RigUnit_CollectionUnion(
     A=/Script/ControlRig.RigUnit_CollectionReplaceItems(New=_r, Old=_l, Items=@RigUnit_CollectionItems.Collection) {is=struct RigVMUnitNode}.Collection
     B=@RigUnit_CollectionItems.Collection
    ) {is=struct RigVMUnitNode}.Collection
   ) {is=struct RigVMUnitNode}
   > n3 = &RigUnit_CollectionLoop(Collection=/Script/ControlRig.RigUnit_CollectionChildren(TypeToSearch=Bone, bIncludeParent=true, bRecursive=true, Parent=@n2.Item) {is=struct RigVMUnitNode}.Collection)
   n4 = /Script/RigVM.RigVMFunction_NameReplace(New=_, Old=_alt_, Name=@n3.Item.Name) {is=struct RigVMUnitNode, template="Replace::Execute(in Name,in Old,in New,out Result)"}
   n5 = /Script/ControlRig.RigUnit_ProjectTransformToNewParent(
    NewParent=(Type=Bone,Name="thumb_01_l")
    OldParent=(Type=Bone,Name="thumb_01_l")
    bChildInitial=True
    bOldParentInitial=True
    Child=@n3.Item
    NewParent.Name=@n4.Result
    OldParent.Name=@n4.Result
   ) {is=struct RigVMUnitNode}
   > &RigUnit_SetRotation(Space=GlobalSpace, Weight=1, Item=@n3.Item, Value=@n5.Transform.Rotation)
   n6 = &RigUnit_ProjectTransformToNewParent(
    Child=(Type=Bone,Name="wolf_toe_l")
    NewParent=(Type=Bone,Name="wolf_ball_l")
    OldParent=(Type=Bone,Name="foot_l")
    bChildInitial=True
    bOldParentInitial=True
   )
   > &RigUnit_SetTransform(Item=(Type=Bone,Name="wolf_toe_l"), Weight=1, Value=@n6.Transform)
   n7 = &RigUnit_ProjectTransformToNewParent(
    Child=(Type=Bone,Name="wolf_toe_r")
    NewParent=(Type=Bone,Name="wolf_ball_r")
    OldParent=(Type=Bone,Name="foot_r")
    bChildInitial=True
    bOldParentInitial=True
   )
   > &RigUnit_SetTransform(Item=(Type=Bone,Name="wolf_toe_r"), Weight=1, Value=@n7.Transform)
   n8 = &RigUnit_BeginExecution()
   n9 = &RigUnit_GetTransform(Item=(Type=Bone,Name="wolf_thigh_l"), bInitial=true)
   n10 = &RigUnit_GetTransform(Item=(Type=Bone,Name="thigh_l"), bInitial=true)
   RigUnit_SetRotation = &RigUnit_SetRotation(
    Item=(Type=Bone,Name="wolf_thigh_l")
    Weight=1
    ExecutePin=@n8.ExecutePin
    Value=&RigVMFunction_MathTransformMakeRelative(
     Parent=&RigVMFunction_MathTransformMakeRelative(Parent=@n9.Transform, Global=@n10.Transform).Local
     Global=&RigUnit_GetTransform(Item=(Type=Bone,Name="thigh_l"), Space=LocalSpace).Transform
    ).Local.Rotation
   )
   n11 = &RigUnit_GetTransform(Item=(Type=Bone,Name="calf_l"), bInitial=true)
   n12 = &RigUnit_GetTransform(Item=(Type=Bone,Name="wolf_calf_l"), bInitial=true)
   > &RigUnit_SetRotation(
    Item=(Type=Bone,Name="wolf_calf_l")
    Weight=1
    Value=&RigVMFunction_MathTransformMakeRelative(
     Parent=&RigVMFunction_MathTransformMakeRelative(Global=@n11.Transform, Parent=@n12.Transform).Local
     Global=&RigUnit_GetTransform(Item=(Type=Bone,Name="calf_l"), Space=LocalSpace).Transform
    ).Local.Rotation
   )
   n13 = &RigUnit_GetTransform(Item=(Type=Bone,Name="wolf_ball_l"), bInitial=true)
   n14 = &RigUnit_GetTransform(Item=(Type=Bone,Name="foot_l"), bInitial=true)
   > &RigUnit_SetRotation(
    Item=(Type=Bone,Name="wolf_ball_l")
    Weight=1
    Value=&RigVMFunction_MathTransformMakeRelative(
     Parent=&RigVMFunction_MathTransformMakeRelative(Parent=@n13.Transform, Global=@n14.Transform).Local
     Global=&RigUnit_GetTransform(Item=(Type=Bone,Name="foot_l"), Space=LocalSpace).Transform
    ).Local.Rotation
   )
   n15 = &RigVMFunction_MathTransformMul(
    A=&RigVMFunction_MathTransformMakeRelative(
     Global=&RigUnit_GetTransform(Item=(Type=Bone,Name="wolf_ball_l"), bInitial=true).Transform
     Parent=&RigUnit_GetTransform(Item=(Type=Bone,Name="foot_l"), bInitial=true).Transform
    ).Local
    B=&RigUnit_GetTransform(Item=(Type=Bone,Name="foot_l")).Transform
   ) {template="Multiply::Execute(in A,in B,out Result)"}
   > /Script/ControlRig.RigUnit_FABRIKPerItem(
    MaxIterations=10
    Precision=1
    Weight=1
    bPropagateToChildren=True
    bSetEffectorTransform=True
    Items=&RigUnit_CollectionItems(Items=((Type=Bone,Name="wolf_thigh_l"),(Type=Bone,Name="wolf_calf_l"),(Type=Bone,Name="wolf_foot_l"),(Type=Bone,Name="wolf_ball_l"))).Collection
    EffectorTransform=@n15.Result
   ) {is=struct RigVMUnitNode}
   > &RigUnit_SetTransform(Item=(Type=Bone,Name="wolf_ball_l"), Weight=1, Value=@n15.Result)
   > /Script/ControlRig.RigUnit_DebugTransformMutableItemSpace(
    Color=(R=1,G=0,B=0,A=1)
    DebugDrawSettings=(DepthPriority=SDPG_Foreground,Lifetime=-1)
    Mode=Axes
    Scale=10
    Space=(Type=Bone)
    WorldOffset=(Rotation=(X=0,Y=0,Z=0,W=1),Translation=(X=0,Y=0,Z=0),Scale3D=(X=1,Y=1,Z=1))
    bEnabled=True
    Transform=@n15.Result
   ) {is=struct RigVMUnitNode}
   > RigUnit_SetTransform = &RigUnit_SetTransform(
    Item=(Type=Bone,Name="wolf_foot_offset_l")
    Weight=1
    Value=&RigUnit_GetTransform(Item=(Type=Bone,Name="wolf_foot_l")).Transform
   )
   RigUnit_AlphaInterp_1 = &RigVMFunction_AlphaInterp(
    InRange=(Max=1)
    InterpSpeedDecreasing=15
    InterpSpeedIncreasing=15
    OutRange=(Max=1)
    Scale=1
    Bias=~
    ClampMin=~
    bMapRange=~
    Value=RigVMVariableNode() {variable=crouch_alpha}.Value
   )
   > &RigUnit_ModifyTransforms(
    ItemToModify=((Item=(Type=Bone,Name="wolf_foot_offset_l"),Transform=(Rotation=(X=0,Y=0,Z=0,W=1),Translation=(X=-4.5,Y=5,Z=-1),Scale3D=(X=1,Y=1,Z=1))))
    ItemToModify.0.Transform.Rotation=&RigVMFunction_MathQuaternionFromEuler(Euler=(X=10,Y=0,Z=17)).Result
    Weight=@RigUnit_AlphaInterp_1.Result
   )
   RigUnit_GetTransform = &RigUnit_GetTransform(Item=(Type=Bone,Name="wolf_foot_offset_l"))
   > /Script/ControlRig.RigUnit_AimItem(
    DebugSettings=(Scale=10,WorldOffset=(Rotation=(X=0,Y=0,Z=0,W=1),Translation=(X=0,Y=0,Z=0),Scale3D=(X=1,Y=1,Z=1)))
    Item=(Type=Bone,Name="wolf_calf_offset_l")
    Primary=(Weight=1,Axis=(X=1,Y=0,Z=0),Target=(X=1,Y=0,Z=0),Kind=Location,Space=(Type=Bone))
    Secondary=(Weight=1,Axis=(X=0,Y=0,Z=0),Target=(X=1,Y=0,Z=0),Kind=Location,Space=(Type=Bone))
    Weight=1
    Primary.Target=@RigUnit_GetTransform.Transform.Translation
   ) {is=struct RigVMUnitNode}
   > &RigUnit_AimItem(
    DebugSettings=(Scale=10,WorldOffset=(Rotation=(X=0,Y=0,Z=0,W=1),Translation=(X=0,Y=0,Z=0),Scale3D=(X=1,Y=1,Z=1)))
    Item=(Type=Bone,Name="wolf_ball_offset_l")
    Primary=(Weight=1,Axis=(X=1,Y=0,Z=0),Target=(X=1,Y=0,Z=0),Kind=Location,Space=(Type=Bone))
    Secondary=(Weight=1,Axis=(X=0,Y=0,Z=0),Target=(X=1,Y=0,Z=0),Kind=Location,Space=(Type=Bone))
    Weight=1
    Primary.Target=@RigUnit_GetTransform.Transform.Translation
   )
   > &RigUnit_AimItem(
    DebugSettings=(Scale=10,WorldOffset=(Rotation=(X=0,Y=0,Z=0,W=1),Translation=(X=0,Y=0,Z=0),Scale3D=(X=1,Y=1,Z=1)))
    Item=(Type=Bone,Name="calf_rope_bot_l")
    Primary=(Weight=1,Axis=(X=1,Y=0,Z=0),Target=(X=1,Y=0,Z=0),Kind=Location,Space=(Type=Bone))
    Secondary=(Weight=1,Axis=(X=0,Y=0,Z=0),Target=(X=1,Y=0,Z=0),Kind=Location,Space=(Type=Bone))
    Weight=1
    Primary.Target=&RigUnit_GetTransform(Item=(Type=Bone,Name="calf_rope_top_l")).Transform.Translation
   )
   > &RigUnit_AimItem(
    DebugSettings=(Scale=10,WorldOffset=(Rotation=(X=0,Y=0,Z=0,W=1),Translation=(X=0,Y=0,Z=0),Scale3D=(X=1,Y=1,Z=1)))
    Item=(Type=Bone,Name="calf_rope_top_l")
    Primary=(Weight=1,Axis=(X=1,Y=0,Z=0),Target=(X=1,Y=0,Z=0),Kind=Location,Space=(Type=Bone))
    Secondary=(Weight=1,Axis=(X=0,Y=0,Z=0),Target=(X=1,Y=0,Z=0),Kind=Location,Space=(Type=Bone))
    Weight=1
    Primary.Target=&RigUnit_GetTransform(Item=(Type=Bone,Name="calf_rope_bot_l")).Transform.Translation
   )
   > &RigUnit_AimItem(
    DebugSettings=(Scale=10,WorldOffset=(Rotation=(X=0,Y=0,Z=0,W=1),Translation=(X=0,Y=0,Z=0),Scale3D=(X=1,Y=1,Z=1)))
    Item=(Type=Bone,Name="calf_band_top_l")
    Primary=(Weight=1,Axis=(X=1,Y=0,Z=0),Target=(X=1,Y=0,Z=0),Kind=Location,Space=(Type=Bone))
    Secondary=(Weight=1,Axis=(X=0,Y=0,Z=0),Target=(X=1,Y=0,Z=0),Kind=Location,Space=(Type=Bone))
    Weight=1
    Primary.Target=&RigUnit_GetTransform(Item=(Type=Bone,Name="calf_band_bot_l")).Transform.Translation
   )
   > &RigUnit_AimItem(
    DebugSettings=(Scale=10,WorldOffset=(Rotation=(X=0,Y=0,Z=0,W=1),Translation=(X=0,Y=0,Z=0),Scale3D=(X=1,Y=1,Z=1)))
    Item=(Type=Bone,Name="calf_band_bot_l")
    Primary=(Weight=1,Axis=(X=1,Y=0,Z=0),Target=(X=1,Y=0,Z=0),Kind=Location,Space=(Type=Bone))
    Secondary=(Weight=1,Axis=(X=0,Y=0,Z=0),Target=(X=1,Y=0,Z=0),Kind=Location,Space=(Type=Bone))
    Weight=1
    Primary.Target=&RigUnit_GetTransform(Item=(Type=Bone,Name="calf_band_top_l")).Transform.Translation
   )
   n16 = &RigUnit_BeginExecution()
   n17 = &RigUnit_GetTransform(Item=(Type=Bone,Name="wolf_thigh_r"), bInitial=true)
   n18 = &RigUnit_GetTransform(Item=(Type=Bone,Name="thigh_r"), bInitial=true)
   &RigUnit_SetRotation(
    Item=(Type=Bone,Name="wolf_thigh_r")
    Weight=1
    ExecutePin=@n16.ExecutePin
    Value=&RigVMFunction_MathTransformMakeRelative(
     Parent=&RigVMFunction_MathTransformMakeRelative(Parent=@n17.Transform, Global=@n18.Transform).Local
     Global=&RigUnit_GetTransform(Item=(Type=Bone,Name="thigh_r"), Space=LocalSpace).Transform
    ).Local.Rotation
   )
   n19 = &RigUnit_GetTransform(Item=(Type=Bone,Name="calf_r"), bInitial=true)
   n20 = &RigUnit_GetTransform(Item=(Type=Bone,Name="wolf_calf_r"), bInitial=true)
   > &RigUnit_SetRotation(
    Item=(Type=Bone,Name="wolf_calf_r")
    Weight=1
    Value=&RigVMFunction_MathTransformMakeRelative(
     Parent=&RigVMFunction_MathTransformMakeRelative(Global=@n19.Transform, Parent=@n20.Transform).Local
     Global=&RigUnit_GetTransform(Item=(Type=Bone,Name="calf_r"), Space=LocalSpace).Transform
    ).Local.Rotation
   )
   n21 = &RigUnit_GetTransform(Item=(Type=Bone,Name="wolf_ball_r"), bInitial=true)
   n22 = &RigUnit_GetTransform(Item=(Type=Bone,Name="foot_r"), bInitial=true)
   > &RigUnit_SetRotation(
    Item=(Type=Bone,Name="wolf_ball_r")
    Weight=1
    Value=&RigVMFunction_MathTransformMakeRelative(
     Parent=&RigVMFunction_MathTransformMakeRelative(Parent=@n21.Transform, Global=@n22.Transform).Local
     Global=&RigUnit_GetTransform(Item=(Type=Bone,Name="foot_r"), Space=LocalSpace).Transform
    ).Local.Rotation
   )
   n23 = &RigVMFunction_MathTransformMul(
    A=&RigVMFunction_MathTransformMakeRelative(
     Global=&RigUnit_GetTransform(Item=(Type=Bone,Name="wolf_ball_r"), bInitial=true).Transform
     Parent=&RigUnit_GetTransform(Item=(Type=Bone,Name="foot_r"), bInitial=true).Transform
    ).Local
    B=&RigUnit_GetTransform(Item=(Type=Bone,Name="foot_r")).Transform
   ) {template="Multiply::Execute(in A,in B,out Result)"}
   > &RigUnit_FABRIKPerItem(
    MaxIterations=10
    Precision=1
    Weight=1
    bPropagateToChildren=True
    bSetEffectorTransform=True
    Items=&RigUnit_CollectionItems(Items=((Type=Bone,Name="wolf_thigh_r"),(Type=Bone,Name="wolf_calf_r"),(Type=Bone,Name="wolf_foot_r"),(Type=Bone,Name="wolf_ball_r"))).Collection
    EffectorTransform=@n23.Result
   )
   > &RigUnit_SetTransform(Item=(Type=Bone,Name="wolf_ball_r"), Weight=1, Value=@n23.Result)
   > &RigUnit_DebugTransformMutableItemSpace(
    Color=(R=1,G=0,B=0,A=1)
    DebugDrawSettings=(DepthPriority=SDPG_Foreground,Lifetime=-1)
    Mode=Axes
    Scale=10
    Space=(Type=Bone)
    WorldOffset=(Rotation=(X=0,Y=0,Z=0,W=1),Translation=(X=0,Y=0,Z=0),Scale3D=(X=1,Y=1,Z=1))
    bEnabled=True
    Transform=@n23.Result
   )
   > &RigUnit_SetTransform(
    Item=(Type=Bone,Name="wolf_foot_offset_r")
    Weight=1
    Value=&RigUnit_GetTransform(Item=(Type=Bone,Name="wolf_foot_r")).Transform
   )
   RigUnit_AlphaInterp_1_1 = &RigVMFunction_AlphaInterp(
    InRange=(Max=1)
    InterpSpeedDecreasing=15
    InterpSpeedIncreasing=15
    OutRange=(Max=1)
    Scale=1
    Bias=~
    ClampMin=~
    bMapRange=~
    Value=RigVMVariableNode() {variable=crouch_alpha}.Value
   )
   > &RigUnit_ModifyTransforms(
    ItemToModify=((Item=(Type=Bone,Name="wolf_foot_offset_r"),Transform=(Rotation=(X=0,Y=0,Z=0,W=1),Translation=(X=4.5,Y=5,Z=1),Scale3D=(X=1,Y=1,Z=1))))
    ItemToModify.0.Transform.Rotation=&RigVMFunction_MathQuaternionFromEuler(Euler=(X=10,Y=0,Z=-22)).Result
    Weight=@RigUnit_AlphaInterp_1_1.Result
   )
   n24 = &RigUnit_GetTransform(Item=(Type=Bone,Name="wolf_foot_offset_r"))
   > &RigUnit_AimItem(
    DebugSettings=(Scale=10,WorldOffset=(Rotation=(X=0,Y=0,Z=0,W=1),Translation=(X=0,Y=0,Z=0),Scale3D=(X=1,Y=1,Z=1)))
    Item=(Type=Bone,Name="wolf_calf_offset_r")
    Primary=(Weight=1,Axis=(X=-1,Y=0,Z=0),Target=(X=1,Y=0,Z=0),Kind=Location,Space=(Type=Bone))
    Secondary=(Weight=1,Axis=(X=0,Y=0,Z=0),Target=(X=1,Y=0,Z=0),Kind=Location,Space=(Type=Bone))
    Weight=1
    Primary.Target=@n24.Transform.Translation
   )
   > &RigUnit_AimItem(
    DebugSettings=(Scale=10,WorldOffset=(Rotation=(X=0,Y=0,Z=0,W=1),Translation=(X=0,Y=0,Z=0),Scale3D=(X=1,Y=1,Z=1)))
    Item=(Type=Bone,Name="wolf_ball_offset_r")
    Primary=(Weight=1,Axis=(X=-1,Y=0,Z=0),Target=(X=1,Y=0,Z=0),Kind=Location,Space=(Type=Bone))
    Secondary=(Weight=1,Axis=(X=0,Y=0,Z=0),Target=(X=1,Y=0,Z=0),Kind=Location,Space=(Type=Bone))
    Weight=1
    Primary.Target=@n24.Transform.Translation
   )
   > &RigUnit_ModifyTransforms(
    ItemToModify=((Item=(Type=Bone,Name="wolf_ball_offset_r"),Transform=(Rotation=(X=0,Y=0,Z=0,W=1),Translation=(X=0,Y=0,Z=0),Scale3D=(X=1,Y=1,Z=1))))
    ItemToModify.0.Transform.Rotation=&RigVMFunction_MathQuaternionFromEuler(Euler=(X=-50,Y=0,Z=0)).Result
    Weight=@RigUnit_AlphaInterp_1_1.Result
   )
   > &RigUnit_AimItem(
    DebugSettings=(Scale=10,WorldOffset=(Rotation=(X=0,Y=0,Z=0,W=1),Translation=(X=0,Y=0,Z=0),Scale3D=(X=1,Y=1,Z=1)))
    Item=(Type=Bone,Name="calf_rope_bot_r")
    Primary=(Weight=1,Axis=(X=-1,Y=0,Z=0),Target=(X=1,Y=0,Z=0),Kind=Location,Space=(Type=Bone))
    Secondary=(Weight=1,Axis=(X=0,Y=0,Z=0),Target=(X=1,Y=0,Z=0),Kind=Location,Space=(Type=Bone))
    Weight=1
    Primary.Target=&RigUnit_GetTransform(Item=(Type=Bone,Name="calf_rope_top_r")).Transform.Translation
   )
   > &RigUnit_AimItem(
    DebugSettings=(Scale=10,WorldOffset=(Rotation=(X=0,Y=0,Z=0,W=1),Translation=(X=0,Y=0,Z=0),Scale3D=(X=1,Y=1,Z=1)))
    Item=(Type=Bone,Name="calf_rope_top_r")
    Primary=(Weight=1,Axis=(X=-1,Y=0,Z=0),Target=(X=1,Y=0,Z=0),Kind=Location,Space=(Type=Bone))
    Secondary=(Weight=1,Axis=(X=0,Y=0,Z=0),Target=(X=1,Y=0,Z=0),Kind=Location,Space=(Type=Bone))
    Weight=1
    Primary.Target=&RigUnit_GetTransform(Item=(Type=Bone,Name="calf_rope_bot_r")).Transform.Translation
   )
   > &RigUnit_AimItem(
    DebugSettings=(Scale=10,WorldOffset=(Rotation=(X=0,Y=0,Z=0,W=1),Translation=(X=0,Y=0,Z=0),Scale3D=(X=1,Y=1,Z=1)))
    Item=(Type=Bone,Name="calf_band_top_r")
    Primary=(Weight=1,Axis=(X=-1,Y=0,Z=0),Target=(X=1,Y=0,Z=0),Kind=Location,Space=(Type=Bone))
    Secondary=(Weight=1,Axis=(X=0,Y=0,Z=0),Target=(X=1,Y=0,Z=0),Kind=Location,Space=(Type=Bone))
    Weight=1
    Primary.Target=&RigUnit_GetTransform(Item=(Type=Bone,Name="calf_band_bot_r")).Transform.Translation
   )
   > &RigUnit_AimItem(
    DebugSettings=(Scale=10,WorldOffset=(Rotation=(X=0,Y=0,Z=0,W=1),Translation=(X=0,Y=0,Z=0),Scale3D=(X=1,Y=1,Z=1)))
    Item=(Type=Bone,Name="calf_band_bot_r")
    Primary=(Weight=1,Axis=(X=-1,Y=0,Z=0),Target=(X=1,Y=0,Z=0),Kind=Location,Space=(Type=Bone))
    Secondary=(Weight=1,Axis=(X=0,Y=0,Z=0),Target=(X=1,Y=0,Z=0),Kind=Location,Space=(Type=Bone))
    Weight=1
    Primary.Target=&RigUnit_GetTransform(Item=(Type=Bone,Name="calf_band_top_r")).Transform.Translation
   )
   n25 = &RigUnit_BeginExecution()
   /Script/ControlRig.RigUnit_SetTranslation(
    Item=(Type=Bone,Name="upperarm_offset_l")
    Space=GlobalSpace
    Weight=1
    bPropagateToChildren=false
    ExecutePin=@n25.ExecutePin
    Value=&RigUnit_GetTransform(Item=(Type=Bone,Name="con_upperarm_l")).Transform.Translation
   ) {is=struct RigVMUnitNode, template="Set Transform::Execute(in Item,in Space,in bInitial,in Value,in Weight,in bPropagateToChildren)"}
   > &RigUnit_AimItem(
    DebugSettings=(Scale=10,WorldOffset=(Rotation=(X=0,Y=0,Z=0,W=1),Translation=(X=0,Y=0,Z=0),Scale3D=(X=1,Y=1,Z=1)))
    Item=(Type=Bone,Name="upperarm_offset_l")
    Primary=(Weight=1,Axis=(X=1,Y=0,Z=0),Target=(X=0,Y=0,Z=0),Kind=Location,Space=(Type=Bone,Name="upperarm_aim_l"))
    Secondary=(Weight=0,Axis=(X=0,Y=0,Z=0),Target=(X=1,Y=0,Z=0),Kind=Location,Space=(Type=Bone))
    Weight=1
    Primary.Axis=&RigUnit_GetRelativeTransformForItem(Child=(Type=Bone,Name="lowerarm_l"), Parent=(Type=Bone,Name="upperarm_offset_l"), bChildInitial=true, bParentInitial=true).RelativeTransform.Translation
   )
   > &RigUnit_AimItem(
    DebugSettings=(Scale=10,WorldOffset=(Rotation=(X=0,Y=0,Z=0,W=1),Translation=(X=0,Y=0,Z=0),Scale3D=(X=1,Y=1,Z=1)))
    Item=(Type=Bone,Name="upperarm_aim_l")
    Primary=(Weight=1,Axis=(X=1,Y=0,Z=0),Target=(X=1,Y=0,Z=0),Kind=Location,Space=(Type=Bone,Name="lowerarm_aim_l"))
    Secondary=(Weight=0,Axis=(X=0,Y=0,Z=0),Target=(X=1,Y=0,Z=0),Kind=Location,Space=(Type=Bone))
    Weight=1
    Primary.Axis=&RigUnit_GetRelativeTransformForItem(Child=(Type=Bone,Name="lowerarm_aim_l"), Parent=(Type=Bone,Name="upperarm_aim_l"), bChildInitial=true, bParentInitial=true).RelativeTransform.Translation
   )
   n26 = &RigUnit_BeginExecution()
   &RigUnit_SetTranslation(
    Item=(Type=Bone,Name="upperarm_offset_r")
    Space=GlobalSpace
    Weight=1
    bPropagateToChildren=false
    ExecutePin=@n26.ExecutePin
    Value=&RigUnit_GetTransform(Item=(Type=Bone,Name="con_upperarm_r")).Transform.Translation
   ) {template="Set Transform::Execute(in Item,in Space,in bInitial,in Value,in Weight,in bPropagateToChildren)"}
   > &RigUnit_AimItem(
    DebugSettings=(Scale=10,WorldOffset=(Rotation=(X=0,Y=0,Z=0,W=1),Translation=(X=0,Y=0,Z=0),Scale3D=(X=1,Y=1,Z=1)))
    Item=(Type=Bone,Name="upperarm_offset_r")
    Primary=(Weight=1,Axis=(X=1,Y=0,Z=0),Target=(X=0,Y=0,Z=0),Kind=Location,Space=(Type=Bone,Name="upperarm_aim_r"))
    Secondary=(Weight=0,Axis=(X=0,Y=0,Z=0),Target=(X=1,Y=0,Z=0),Kind=Location,Space=(Type=Bone))
    Weight=1
    Primary.Axis=&RigUnit_GetRelativeTransformForItem(Child=(Type=Bone,Name="lowerarm_r"), Parent=(Type=Bone,Name="upperarm_offset_r"), bChildInitial=true, bParentInitial=true).RelativeTransform.Translation
   )
   > &RigUnit_AimItem(
    DebugSettings=(Scale=10,WorldOffset=(Rotation=(X=0,Y=0,Z=0,W=1),Translation=(X=0,Y=0,Z=0),Scale3D=(X=1,Y=1,Z=1)))
    Item=(Type=Bone,Name="upperarm_aim_r")
    Primary=(Weight=1,Axis=(X=1,Y=0,Z=0),Target=(X=1,Y=0,Z=0),Kind=Location,Space=(Type=Bone,Name="lowerarm_aim_r"))
    Secondary=(Weight=0,Axis=(X=0,Y=0,Z=0),Target=(X=1,Y=0,Z=0),Kind=Location,Space=(Type=Bone))
    Weight=1
    Primary.Axis=&RigUnit_GetRelativeTransformForItem(Child=(Type=Bone,Name="lowerarm_aim_r"), Parent=(Type=Bone,Name="upperarm_aim_r"), bChildInitial=true, bParentInitial=true).RelativeTransform.Translation
   )
   n27 = /Script/ControlRig.RigUnit_GetInitialBoneTransform(Bone=thigh_r, Space=GlobalSpace) {is=struct RigVMUnitNode}
   n28 = &RigUnit_GetInitialBoneTransform(Bone=wolf_thigh_r, Space=GlobalSpace)
   /Script/ControlRig.RigUnit_SetBoneRotation(
    Bone=wolf_thigh_r
    ExecutePin=()
    Space=LocalSpace
    Weight=1
    bPropagateToChildren=True
    Rotation=&RigVMFunction_MathTransformMakeRelative(
     Global=&RigUnit_GetBoneTransform(Bone=thigh_r, Space=LocalSpace).Transform
     Parent=&RigVMFunction_MathTransformMakeRelative(Global=@n27.Transform, Parent=@n28.Transform).Local
    ).Local.Rotation
   ) {is=struct RigVMUnitNode}
   n29 = &RigUnit_GetInitialBoneTransform(Bone=calf_r, Space=GlobalSpace)
   n30 = &RigUnit_GetInitialBoneTransform(Bone=wolf_calf_r, Space=GlobalSpace)
   > &RigUnit_SetBoneRotation(
    Bone=wolf_calf_r
    Space=LocalSpace
    Weight=1
    bPropagateToChildren=True
    Rotation=&RigVMFunction_MathTransformMakeRelative(
     Global=&RigUnit_GetBoneTransform(Bone=calf_r, Space=LocalSpace).Transform
     Parent=&RigVMFunction_MathTransformMakeRelative(Global=@n29.Transform, Parent=@n30.Transform).Local
    ).Local.Rotation
   )
   n31 = &RigUnit_GetInitialBoneTransform(Bone=foot_r, Space=GlobalSpace)
   n32 = &RigUnit_GetInitialBoneTransform(Bone=wolf_ball_r, Space=GlobalSpace)
   > &RigUnit_SetBoneRotation(
    Bone=wolf_ball_r
    Space=LocalSpace
    Weight=1
    bPropagateToChildren=True
    Rotation=&RigVMFunction_MathTransformMakeRelative(
     Global=&RigUnit_GetBoneTransform(Bone=foot_r, Space=LocalSpace).Transform
     Parent=&RigVMFunction_MathTransformMakeRelative(Global=@n31.Transform, Parent=@n32.Transform).Local
    ).Local.Rotation
   )
   n33 = &RigVMFunction_MathTransformMul(
    B=&RigUnit_GetBoneTransform(Bone=foot_r).Transform
    A=&RigVMFunction_MathTransformMakeRelative(
     Parent=&RigUnit_GetInitialBoneTransform(Bone=foot_r, Space=GlobalSpace).Transform
     Global=&RigUnit_GetInitialBoneTransform(Bone=wolf_ball_r, Space=GlobalSpace).Transform
    ).Local
   ) {template="Multiply::Execute(in A,in B,out Result)"}
   > /Script/ControlRig.RigUnit_FABRIK(
    EffectorBone=wolf_ball_r
    MaxIterations=10
    Precision=1
    StartBone=wolf_thigh_r
    Weight=1
    bPropagateToChildren=True
    bSetEffectorTransform=True
    EffectorTransform=@n33.Result
   ) {is=struct RigVMUnitNode}
   > /Script/ControlRig.RigUnit_SetBoneTransform(Bone=wolf_ball_r, Space=GlobalSpace, Weight=1, bPropagateToChildren=True, Transform=@n33.Result) {is=struct RigVMUnitNode}
   > /Script/ControlRig.RigUnit_DebugTransformMutable(
    Color=(R=1,G=0,B=0,A=1)
    DebugDrawSettings=(DepthPriority=SDPG_Foreground,Lifetime=-1)
    Mode=Axes
    Scale=10
    Space=None
    Thickness=0.5
    WorldOffset=(Rotation=(X=0,Y=0,Z=0,W=1),Translation=(X=0,Y=0,Z=0),Scale3D=(X=1,Y=1,Z=1))
    bEnabled=True
    Transform=@n33.Result
   ) {is=struct RigVMUnitNode}
   &RigUnit_BeginExecution()
   &RigVMFunction_MathFloatRemap(
    SourceMaximum=65
    SourceMinimum=25
    TargetMinimum=0
    bClamp=true
    Value=&RigVMFunction_MathQuaternionToEuler(RotationOrder=XZY, Value=&RigUnit_GetRelativeBoneTransform(Bone=calf_l, Space=thigh_l).Transform.Rotation).Result.Z
   )
   &RigVMFunction_MathFloatRemap(
    SourceMaximum=65
    SourceMinimum=25
    TargetMinimum=0
    bClamp=true
    Value=&RigVMFunction_MathQuaternionToEuler(RotationOrder=XZY, Value=&RigUnit_GetRelativeBoneTransform(Bone=calf_l, Space=thigh_l).Transform.Rotation).Result.Z
   )
   &RigUnit_SetTransform(
    ExecutePin=()
    Item=(Type=Bone,Name="thumb_alt_01_l")
    Value=(Rotation=(X=-0.1456,Y=-0.7021,Z=-0.3074,W=-0.6255),Translation=(X=52.6316,Y=10.1738,Z=100.8842),Scale3D=(X=1,Y=1,Z=1))
    Weight=1
    Value.Rotation=@n5.Transform.Rotation
   )
  ]}
]
hierarchy [
 root {loc=<0 0 0>, rot=<0 0 0 1>}
 pelvis {parent=root, loc=<0 -0.7071 93.1191>, rot=<0.0127 0.707 -0.0127 -0.707>}
 spine_01 {parent=pelvis, loc=<6.7494 -0.6388 0>, rot=<0 0 0.034 -0.9994>}
 spine_02 {parent=spine_01, loc=<7.2518 -0.035 0>, rot=<0 0 0.0238 -0.9997>}
 spine_03 {parent=spine_02, loc=<6.0104 0 0>, rot=<0 0 0.0149 -0.9999>}
 spine_04 {parent=spine_03, loc=<8.1024 0 0>, rot=<0 0 -0.0132 -0.9999>}
 spine_05 {parent=spine_04, loc=<10.0146 0 0>, rot=<0 0 -0.0015 -1>}
 clavicle_l {parent=spine_05, loc=<4.3575 -0.1466 -2.3223>, rot=<0.0465 -0.7289 -0.009 -0.6829>}
 upperarm_l {parent=clavicle_l, loc=<15.5888 -0.5667 2.0053>, rot=<0.0226 -0.3234 -0.0088 -0.9459>}
 lowerarm_l {parent=upperarm_l, loc=<27.5375 1.4837 0.0149>, rot=<0.0391 -0.0099 -0.2453 -0.9686>}
 hand_l {parent=lowerarm_l, loc=<22.5509 -0.6357 -1.4962>, rot=<-0.7081 -0.0507 0.0975 -0.6975>}
 index_metacarpal_l {parent=hand_l, loc=<1.6535 0.0183 -2.96>, rot=<0.0158 -0.0124 -0.0173 -0.9996>}
 index_01_l {parent=index_metacarpal_l, loc=<8.0395 -0.3306 -0.0779>, rot=<-0.0095 0.0477 0.1976 -0.9791>}
 index_02_l {parent=index_01_l, loc=<3.0921 -0.8255 0.0424>, rot=<-0.0024 0.0024 0.4511 -0.8925>}
 index_alt_01_l {parent=index_metacarpal_l, loc=<9.3282 -1.1553 0.1622>, rot=<-0.073 -0.015 0.2318 -0.9699>}
 index_alt_02_l {parent=index_alt_01_l, loc=<5.3609 0 0>, rot=<0 0 0.3173 -0.9483>}
 ring_metacarpal_l {parent=hand_l, loc=<2.2074 -0.3625 1.3638>, rot=<-0.0284 0.0722 -0.0458 -0.9959>}
 ring_01_l {parent=ring_metacarpal_l, loc=<6.8594 -0.8315 -0.1617>, rot=<-0.0472 -0.0282 0.2509 -0.9665>}
 ring_02_l {parent=ring_01_l, loc=<3.0299 -0.5829 -0.1026>, rot=<0 0 0.5377 -0.8432>}
 ring_alt_01_l {parent=ring_metacarpal_l, loc=<8.5721 -1.8967 0.2974>, rot=<-0.0534 -0.06 0.2852 -0.9551>}
 ring_alt_02_l {parent=ring_alt_01_l, loc=<5.102 0.1533 -0.0322>, rot=<0 0 0.2806 -0.9598>}
 thumb_01_l {parent=hand_l, loc=<1.1081 -0.6318 -2.4896>, rot=<0.7851 -0.4959 0.075 -0.3636>}
 thumb_02_l {parent=thumb_01_l, loc=<5.9511 0.0111 0.0403>, rot=<-0.0034 0.0213 0.285 -0.9583>}
 thumb_alt_01_l {parent=hand_l, loc=<3.6718 0.2244 -4.8825>, rot=<0.5672 -0.3555 0.0693 -0.7396>}
 thumb_alt_02_l {parent=thumb_alt_01_l, loc=<5.1661 0.3631 0.0374>, rot=<0 0 0.2983 -0.9545>}
 dyn_parent_arm_fthr_l {parent=lowerarm_l, loc=<20.439 -0.5086 -1.1621>, rot=<0 0 0 -1>}
 dyn_arm_fthrA_01_l {parent=dyn_parent_arm_fthr_l, loc=<-3.9672 -7.2453 -0.2168>, rot=<-0.0743 -0.2159 0.7846 -0.5765>}
 dyn_arm_fthrA_02_l {parent=dyn_arm_fthrA_01_l, loc=<4.2805 0 0.341>, rot=<-0.004 -0.0608 0.0659 -0.996>}
 dyn_hinge_arm_fthr_l {parent=dyn_parent_arm_fthr_l, loc=<-5.6872 -7.6859 -0.1622>, rot=<0.0657 -0.1086 0.9911 -0.0391>}
 dyn_arm_fthrB_01_l {parent=dyn_hinge_arm_fthr_l, loc=<3.0574 1.5063 -0.0768>, rot=<0.0852 0.0186 -0.2967 -0.951>}
 dyn_arm_fthrB_02_l {parent=dyn_arm_fthrB_01_l, loc=<4.5886 0 -0.0204>, rot=<-0.0013 0.0219 0.0574 -0.9981>}
 dyn_arm_fthrB_03_l {parent=dyn_arm_fthrB_02_l, loc=<4.3569 0 0>, rot=<0 0 0 1>}
 dyn_arm_fthrC_01_l {parent=dyn_hinge_arm_fthr_l, loc=<11.359 1.3795 0.1543>, rot=<0.0115 0.0395 -0.172 -0.9842>}
 dyn_arm_fthrC_02_l {parent=dyn_arm_fthrC_01_l, loc=<5.3773 0 0>, rot=<0.0182 0.0207 0.1568 -0.9873>}
 lowerarm_aim_l {parent=lowerarm_l, loc=<19.716 0 -0.8214>, rot=<0 -0.0208 0 -0.9998>}
 upperarm_offset_l {parent=upperarm_l, loc=<0 0 0>, rot=<0 0 0 -1>}
 upperarm_aim_l {parent=upperarm_offset_l, loc=<27.7596 1.2755 -0.2454>, rot=<0.0442 -0.0301 -0.2444 -0.9682>}
 con_lowerarm_l {parent=upperarm_l, loc=<27.7596 1.2755 -0.2454>, rot=<0.0391 -0.0099 -0.2453 -0.9686>}
 clavicle_r {parent=spine_05, loc=<4.3575 -0.1466 2.3223>, rot=<0.7289 0.0465 -0.6829 0.009>}
 upperarm_r {parent=clavicle_r, loc=<-15.5888 0.5667 -2.0055>, rot=<0.0226 -0.3234 -0.0088 -0.9459>}
 lowerarm_r {parent=upperarm_r, loc=<-27.5375 -1.4837 -0.0148>, rot=<0.0391 -0.0099 -0.2453 -0.9686>}
 hand_r {parent=lowerarm_r, loc=<-22.5515 0.6359 1.4968>, rot=<-0.7081 -0.0507 0.0975 -0.6975>}
 index_metacarpal_r {parent=hand_r, loc=<-1.6534 -0.0181 2.9601>, rot=<0.0158 -0.0124 -0.0173 -0.9996>}
 index_01_r {parent=index_metacarpal_r, loc=<-8.0394 0.3307 0.078>, rot=<-0.0095 0.0477 0.1976 -0.9791>}
 index_02_r {parent=index_01_r, loc=<-3.0921 0.8256 -0.0424>, rot=<-0.0024 0.0024 0.4511 -0.8925>}
 index_alt_01_r {parent=index_metacarpal_r, loc=<-9.3279 1.1547 -0.1622>, rot=<-0.073 -0.015 0.2318 -0.9699>}
 index_alt_02_r {parent=index_alt_01_r, loc=<-5.3609 0 0>, rot=<0 0 0.3173 -0.9483>}
 ring_metacarpal_r {parent=hand_r, loc=<-2.2073 0.3627 -1.3638>, rot=<-0.0284 0.0722 -0.0458 -0.9959>}
 ring_01_r {parent=ring_metacarpal_r, loc=<-6.8595 0.8316 0.1618>, rot=<-0.0472 -0.0282 0.2509 -0.9665>}
 ring_02_r {parent=ring_01_r, loc=<-3.0294 0.5831 0.1027>, rot=<0 0 0.5377 -0.8432>}
 ring_alt_01_r {parent=ring_metacarpal_r, loc=<-8.5718 1.8963 -0.2973>, rot=<-0.0534 -0.06 0.2852 -0.9551>}
 ring_alt_02_r {parent=ring_alt_01_r, loc=<-5.102 -0.1533 0.0322>, rot=<0 0 0.2806 -0.9598>}
 thumb_01_r {parent=hand_r, loc=<-1.1082 0.6319 2.4896>, rot=<0.7851 -0.4959 0.075 -0.3636>}
 thumb_02_r {parent=thumb_01_r, loc=<-5.951 -0.0112 -0.0404>, rot=<-0.0034 0.0213 0.285 -0.9583>}
 thumb_alt_01_r {parent=hand_r, loc=<-3.6716 -0.2245 4.8826>, rot=<0.5672 -0.3555 0.0693 -0.7396>}
 thumb_alt_02_r {parent=thumb_alt_01_r, loc=<-5.1659 -0.3632 -0.0374>, rot=<0 0 0.2983 -0.9545>}
 dyn_parent_arm_fthr_r {parent=lowerarm_r, loc=<-20.4389 0.5085 1.162>, rot=<0 0 0 1>}
 dyn_arm_fthrA_01_r {parent=dyn_parent_arm_fthr_r, loc=<3.9669 7.2454 0.2172>, rot=<0.7846 0.5765 0.0743 -0.2159>}
 dyn_arm_fthrA_02_r {parent=dyn_arm_fthrA_01_r, loc=<4.2804 0.0001 0.3411>, rot=<0.004 -0.0608 -0.0659 -0.996>}
 dyn_hinge_arm_fthr_r {parent=dyn_parent_arm_fthr_r, loc=<5.6868 7.686 0.1627>, rot=<0.0657 -0.1086 0.9911 -0.0391>}
 dyn_arm_fthrB_01_r {parent=dyn_hinge_arm_fthr_r, loc=<-3.0576 -1.5062 0.0767>, rot=<0.2967 -0.951 0.0852 -0.0186>}
 dyn_arm_fthrB_02_r {parent=dyn_arm_fthrB_01_r, loc=<4.5885 -0.0002 -0.0206>, rot=<0.0013 0.0219 -0.0574 -0.9981>}
 dyn_arm_fthrC_01_r {parent=dyn_hinge_arm_fthr_r, loc=<-11.3594 -1.3792 -0.1547>, rot=<0.172 -0.9842 0.0115 -0.0395>}
 dyn_arm_fthrC_02_r {parent=dyn_arm_fthrC_01_r, loc=<5.3771 -0.0002 -0.0002>, rot=<-0.0182 0.0207 -0.1568 -0.9873>}
 lowerarm_aim_r {parent=lowerarm_r, loc=<-19.7164 0.0002 0.8219>, rot=<0 -0.0208 0 -0.9998>}
 upperarm_offset_r {parent=upperarm_r, loc=<0 0 0>, rot=<0 0 0 1>}
 upperarm_aim_r {parent=upperarm_offset_r, loc=<-27.76 -1.2755 0.2457>, rot=<0.0442 -0.0301 -0.2444 -0.9682>}
 neck_01 {parent=spine_05, loc=<11.3138 0.1529 0>, rot=<0 0 -0.2003 -0.9797>}
 neck_02 {parent=neck_01, loc=<4.5627 0 0>, rot=<0 0 0.0504 -0.9987>}
 head {parent=neck_02, loc=<5.3914 0 0>, rot=<0 0 0.115 -0.9934>}
 con_upperarm_l {parent=spine_05, loc=<5.418 -1.422 -17.9621>, rot=<0.0559 -0.9106 -0.016 -0.4091>}
 con_upperarm_r {parent=spine_05, loc=<5.2545 -1.3239 17.8549>, rot=<0.9106 0.0559 -0.4091 0.016>}
 dyn_parent_fthr_bckA_01 {parent=spine_04, loc=<-0.3261 -7.5333 9.3336>, rot=<0.0648 0.1946 0.966 -0.1574>}
 dyn_fthr_bckA_01 {parent=dyn_parent_fthr_bckA_01, loc=<0 0 0>, rot=<0 0 0 1>}
 dyn_fthr_bckA_02 {parent=dyn_fthr_bckA_01, loc=<6.0255 -0.26 0>, rot=<0 0 0.1174 -0.9931>}
 thigh_l {parent=pelvis, loc=<-2.5693 -0.0109 -7.7501>, rot=<0.0012 0.0572 0.0155 -0.9982>}
 calf_l {parent=thigh_l, loc=<-39.0484 0 0>, rot=<0.0006 -0.0161 -0.0391 -0.9991>}
 foot_l {parent=calf_l, loc=<-42.561 0.1077 0.4292>, rot=<0.0418 -0.0455 0.0343 -0.9975>}
 ball_l {parent=foot_l, loc=<-6.1599 13.8264 0.6806>, rot=<0.0001 -0.0001 -0.7063 -0.7079>}
 thigh_twist_01_l {parent=thigh_l, loc=<-21.6303 0 0>, rot=<0.0013 0 0 -1>}
 thigh_r {parent=pelvis, loc=<-2.5691 -0.0109 7.7501>, rot=<0.0572 -0.0012 0.9982 0.0155>}
 calf_r {parent=thigh_r, loc=<39.0487 0 0>, rot=<0.0006 -0.0161 -0.0391 -0.9991>}
 foot_r {parent=calf_r, loc=<42.5612 -0.1077 -0.4292>, rot=<0.0418 -0.0455 0.0343 -0.9975>}
 ball_r {parent=foot_r, loc=<6.1598 -13.8264 -0.6806>, rot=<0.0001 -0.0001 -0.7063 -0.7079>}
 thigh_twist_01_r {parent=thigh_r, loc=<21.6303 0 0>, rot=<0.0013 0 0 -1>}
 wolf_thigh_l {parent=pelvis, loc=<-8.792 1.4966 -8.4711>, rot=<0.0012 0.0572 0.0155 -0.9982>}
 wolf_calf_l {parent=wolf_thigh_l, loc=<-30.7835 6.8969 0.7336>, rot=<-0.0024 -0.0205 -0.0398 -0.999>}
 wolf_foot_l {parent=wolf_calf_l, loc=<-29.3538 -13.5158 0.4328>, rot=<-0.0017 0.0007 0.0299 -0.9996>}
 wolf_ball_l {parent=wolf_foot_l, loc=<-17.8117 5.4629 -0.3945>, rot=<0.0477 -0.0404 0.0056 -0.998>}
 wolf_toe_l {parent=wolf_ball_l, loc=<-5.5882 6.1704 0.2799>, rot=<-0.0059 0.0622 -0.7011 -0.7103>}
 toe_claw_l {parent=wolf_ball_l, loc=<-5.665 -3.658 -0.3804>, rot=<0.3991 -0.5584 0.5413 -0.4858>}
 wolf_ball_offset_l {parent=wolf_ball_l, loc=<0 0 0>, rot=<-0.0524 0.0531 0.1397 -0.9874>}
 calf_rope_bot_l {parent=wolf_ball_offset_l, loc=<3.8194 3.8019 -0.1874>, rot=<0.0017 -0.0198 -0.1782 -0.9838>}
 wolf_foot_offset_l {parent=wolf_calf_l, loc=<-29.3538 -13.5158 0.4328>, rot=<-0.0017 0.0007 0.0299 -0.9996>}
 calf_band_bot_l {parent=wolf_foot_offset_l, loc=<10.2931 -3.316 0.2633>, rot=<-0.0032 -0.016 -0.359 -0.9332>}
 calf_band_top_l {parent=wolf_calf_l, loc=<-4.3312 -5.4781 0.2335>, rot=<0.331 -0.9435 -0.005 0.0146>}
 wolf_calf_offset_l {parent=wolf_thigh_l, loc=<-30.7835 6.8969 0.7336>, rot=<0.3108 -0.9502 -0.0072 0.0215>}
 calf_rope_top_l {parent=wolf_calf_offset_l, loc=<13.9287 3.3016 0.0162>, rot=<-0.0009 0.0024 -0.2688 -0.9632>}
 wolf_thigh_r {parent=pelvis, loc=<-8.7921 1.4966 8.4711>, rot=<0.0572 -0.0012 0.9982 0.0155>}
 wolf_calf_r {parent=wolf_thigh_r, loc=<30.7835 -6.8968 -0.7336>, rot=<-0.0024 -0.0205 -0.0398 -0.999>}
 wolf_foot_r {parent=wolf_calf_r, loc=<29.3538 13.5158 -0.4328>, rot=<-0.0017 0.0007 0.0299 -0.9996>}
 wolf_ball_r {parent=wolf_foot_r, loc=<17.8117 -5.4629 0.3945>, rot=<0.0477 -0.0404 0.0056 -0.998>}
 wolf_toe_r {parent=wolf_ball_r, loc=<5.5882 -6.1704 -0.28>, rot=<-0.0059 0.0622 -0.7011 -0.7103>}
 toe_claw_r {parent=wolf_ball_r, loc=<5.665 3.658 0.3803>, rot=<0.3991 -0.5584 0.5413 -0.4858>}
 wolf_ball_offset_r {parent=wolf_ball_r, loc=<0 0 0>, rot=<-0.0524 0.0531 0.1397 -0.9874>}
 calf_rope_bot_r {parent=wolf_ball_offset_r, loc=<-3.8194 -3.8019 0.1873>, rot=<0.0017 -0.0198 -0.1782 -0.9838>}
 calf_band_top_r {parent=wolf_calf_r, loc=<4.3312 5.4781 -0.2335>, rot=<0.331 -0.9435 -0.005 0.0146>}
 wolf_foot_offset_r {parent=wolf_calf_r, loc=<29.3538 13.5158 -0.4328>, rot=<-0.0017 0.0007 0.0299 -0.9996>}
 calf_band_bot_r {parent=wolf_foot_offset_r, loc=<-10.2932 3.316 -0.2633>, rot=<-0.0032 -0.016 -0.359 -0.9332>}
 wolf_calf_offset_r {parent=wolf_thigh_r, loc=<30.7835 -6.8968 -0.7336>, rot=<0.3108 -0.9502 -0.0072 0.0215>}
 calf_rope_top_r {parent=wolf_calf_offset_r, loc=<-13.9287 -3.3016 -0.0162>, rot=<-0.0009 0.0024 -0.2688 -0.9632>}
 pelvis_offset {parent=pelvis, loc=<0 0 0>, rot=<0 0 0 -1>}
 dyn_parent_skirt_l {parent=pelvis_offset, loc=<2.6526 2.0557 -13.2931>, rot=<-0.0908 -0.6717 0.7339 -0.045>}
 dyn_driver_skirt_l {parent=dyn_parent_skirt_l, loc=<4.2997 0 0>, rot=<0 0 0 -1>}
 dyn_skirt_01_l {parent=dyn_driver_skirt_l, loc=<0 0 0>, rot=<0 0 0 -1>}
 dyn_skirt_02_l {parent=dyn_skirt_01_l, loc=<7.7296 -0.0419 0>, rot=<-0.0002 -0.0007 0.0136 -0.9999>}
 dyn_skirt_03_l {parent=dyn_skirt_02_l, loc=<7.5289 0 0>, rot=<0 0 0.0588 -0.9983>}
 dyn_skirt_04_l {parent=dyn_skirt_03_l, loc=<7.2007 0 0>, rot=<0 0 0.0509 -0.9987>}
 dyn_beltA_l {parent=pelvis_offset, loc=<0.0883 0.211 -5.894>, rot=<-0.4067 0.4932 -0.5548 -0.5324>}
 dyn_beltA_01_l {parent=pelvis_offset, loc=<0.0883 0.211 -5.894>, rot=<0.4491 -0.5047 -0.5329 -0.5095>}
 dyn_beltA_01_r {parent=pelvis_offset, loc=<0.0883 0.211 5.894>, rot=<-0.4491 0.5047 -0.5329 -0.5095>}
 dyn_parent_skirt_r {parent=pelvis_offset, loc=<-0.4363 0.0383 14.0331>, rot=<0.0775 0.6707 0.7354 -0.0572>}
 dyn_driver_skirt_r {parent=dyn_parent_skirt_r, loc=<0 0 0>, rot=<0 0 0 -1>}
 dyn_skirt_01_r {parent=dyn_driver_skirt_r, loc=<0 0 0>, rot=<0 0 0 -1>}
 dyn_skirt_02_r {parent=dyn_skirt_01_r, loc=<7.7296 -0.0419 0>, rot=<0.0002 0.0007 0.0136 -0.9999>}
 dyn_skirt_03_r {parent=dyn_skirt_02_r, loc=<7.5289 0 0>, rot=<0 0 0.0588 -0.9983>}
 dyn_skirt_04_r {parent=dyn_skirt_03_r, loc=<7.2006 0 0>, rot=<0 0 0.0509 -0.9987>}
 dyn_parent_skirt_bck {parent=pelvis_offset, loc=<-0.5135 -8.3302 -0.4204>, rot=<-0.0008 0.0003 0.9999 0.0116>}
 dyn_driver_skirt_bck {parent=dyn_parent_skirt_bck, loc=<0 0 0>, rot=<0 0 0 1>}
 dyn_skirt_bck_01 {parent=dyn_driver_skirt_bck, loc=<0 0 0>, rot=<0 0 0 1>}
 dyn_skirt_bck_02 {parent=dyn_skirt_bck_01, loc=<7.7296 -0.0419 0>, rot=<0 0 0 1>}
 dyn_skirt_bck_03 {parent=dyn_skirt_bck_02, loc=<7.5289 0 0>, rot=<0 0 0 1>}
 dyn_skirt_bck_04 {parent=dyn_skirt_bck_03, loc=<7.2007 0 0>, rot=<0 0 0 1>}
 dyn_skirt_bck_05 {parent=dyn_skirt_bck_04, loc=<6.4856 0 0>, rot=<0 0 0 1>}
 dyn_parent_skirt_frt {parent=pelvis_offset, loc=<-0.0662 10.8658 -0.4329>, rot=<-0.0008 0.0003 0.9999 0.0116>}
 dyn_driver_skirt_frt {parent=dyn_parent_skirt_frt, loc=<0 0 0>, rot=<0 0 0.0817 -0.9967>}
 dyn_skirt_frt_01 {parent=dyn_driver_skirt_frt, loc=<0 0 0>, rot=<0 0 0 1>}
 dyn_skirt_frt_02 {parent=dyn_skirt_frt_01, loc=<7.7296 -0.0419 0>, rot=<0 0 0 1>}
 dyn_skirt_frt_03 {parent=dyn_skirt_frt_02, loc=<7.5289 0 0>, rot=<0 0 0 1>}
 dyn_skirt_frt_04 {parent=dyn_skirt_frt_03, loc=<7.2007 0 0>, rot=<0 0 0 1>}
 dyn_skirt_frt_05 {parent=dyn_skirt_frt_04, loc=<6.4856 0 0>, rot=<0 0 0 1>}
 dyn_parent_skirt_fl {parent=pelvis_offset, loc=<2.8443 8.3583 -9.5756>, rot=<-0.1587 -0.9626 0.2084 -0.0699>}
 dyn_driver_skirt_fl {parent=dyn_parent_skirt_fl, loc=<2.6103 0 0>, rot=<0 0 0 -1>}
 dyn_skirt_01_fl {parent=dyn_driver_skirt_fl, loc=<0 0 0>, rot=<0 0 0 -1>}
 dyn_thigh_cor {parent=dyn_skirt_01_fl, loc=<5.0635 -0.123 0.0377>, rot=<0.035 -0.0193 0.0561 -0.9976>}
 dyn_parent_skirt_fAl {parent=pelvis_offset, loc=<2.7463 5.9676 -12.4093>, rot=<-0.1327 -0.823 0.551 -0.0379>}
 dyn_driver_skirt_fAl {parent=dyn_parent_skirt_fAl, loc=<2.6103 0 0>, rot=<0 0 0 -1>}
 dyn_skirt_01_fAl {parent=dyn_driver_skirt_fAl, loc=<0 0 0>, rot=<0 0 0 -1>}
 dyn_parent_skirt_fr {parent=pelvis_offset, loc=<3.2867 7.322 8.9212>, rot=<-0.1706 -0.9768 -0.1172 0.055>}
 dyn_driver_skirt_fr {parent=dyn_parent_skirt_fr, loc=<2.6104 0 0>, rot=<0 0 0 -1>}
 dyn_skirt_01_fr {parent=dyn_driver_skirt_fr, loc=<0 0 0>, rot=<0 0 0 -1>}
 dyn_thigh_cor_r {parent=dyn_skirt_01_fr, loc=<6.5489 -0.2926 0.0141>, rot=<-0.035 0.0193 0.0561 -0.9976>}
 dyn_parent_skirt_frA {parent=pelvis_offset, loc=<3.3712 4.9315 11.6239>, rot=<-0.1453 -0.8055 -0.5707 0.0662>}
 dyn_skirt_01_frA {parent=dyn_parent_skirt_frA, loc=<2.6104 0 0>, rot=<0 0 0 -1>}
 dyn_thigh_cor_ra {parent=dyn_skirt_01_frA, loc=<6.5489 -0.2926 0.0141>, rot=<-0.035 0.0193 0.0561 -0.9976>}
]
variables [crouch_alpha {type=double, default=0}]
