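# Physics asset for the feather bones of the male medium Dark Minion body cosmetic, covering both arms and the back.
# Capsule bodies form two-bone chains: A feathers hang from each arm's parent bone, B and C feathers from an intermediate hinge bone.
# Hinge joints leave swing and twist free, while feather joints set narrow limits, asymmetric between left and right.
# The back feather is a separate two-bone chain on its own parent with tight limits.
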PhysicsAsset /BRCosmetics/Characters/Player/Male/Medium/Bodies/M_MED_Dark_Minion/Meshes/M_MED_Dark_Minion_Fthrs_Physics
bodies {
 dyn_arm_fthrA_01_l [{shapeType=Capsule, length=1.75, radius=2.5, center=<2 0 0>, rotation=<rot 90 0 0>}]
 dyn_arm_fthrA_01_r [{shapeType=Capsule, length=2.25, radius=2, center=<3 0 0>, rotation=<rot 90 0 0>}]
 dyn_arm_fthrA_02_l [{shapeType=Capsule, length=2.25, radius=1.25, center=<2 0 0>, rotation=<rot 90 0 0>}]
 dyn_arm_fthrA_02_r [{shapeType=Capsule, length=2.25, radius=2, center=<3 0 0>, rotation=<rot 90 0 0>}]
 dyn_arm_fthrB_01_l [{shapeType=Capsule, length=2.75, radius=3.25, center=<1 0 0>, rotation=<rot 90 0 0>}]
 dyn_arm_fthrB_01_r [{shapeType=Capsule, length=2.5, radius=3, center=<1 0 0>, rotation=<rot 90 0 0>}]
 dyn_arm_fthrB_02_l [{shapeType=Capsule, length=3.5, radius=2, center=<3 0 0>, rotation=<rot 90 0 0>}]
 dyn_arm_fthrB_02_r [{shapeType=Capsule, length=1.75, radius=2.25, center=<3 0 0>, rotation=<rot 90 0 0>}]
 dyn_arm_fthrC_01_l [{shapeType=Capsule, length=2.25, radius=2, center=<4 0 0>, rotation=<rot 90 0 0>}]
 dyn_arm_fthrC_01_r [{shapeType=Capsule, length=2.25, radius=1.75, center=<4 0 0>, rotation=<rot 90 0 0>}]
 dyn_arm_fthrC_02_l [{shapeType=Capsule, length=1.75, radius=1.75, center=<4 0 0>, rotation=<rot 90 0 0>}]
 dyn_arm_fthrC_02_r [{shapeType=Capsule, length=2, radius=1.5, center=<5 0 0>, rotation=<rot 90 0 0>}]
 dyn_fthr_bckA_01 [{shapeType=Capsule, length=4.5, radius=3, center=<5 0 0>, rotation=<rot 90 0 0>}]
 dyn_fthr_bckA_02 [{shapeType=Capsule, length=2.75, radius=2.5, center=<5 0 0>, rotation=<rot 90 0 0>}]
 dyn_hinge_arm_fthr_l [{shapeType=Capsule, length=1, radius=1}]
 dyn_hinge_arm_fthr_r [{shapeType=Capsule, length=1.5, radius=1, center=<3 0.7 0.1716>, rotation=<rot 0 -95.7501 -89.5259>}]
 dyn_parent_arm_fthr_l [{shapeType=Capsule, length=0.505, radius=0.505, center=<-4.3925 -6.9516 -0.3603>, rotation=<rot -12.0105 -90 4.7852>}]
 dyn_parent_arm_fthr_r [{shapeType=Capsule, length=1.5, radius=1, center=<3 0 0>, rotation=<rot 0 -90 0>}]
 dyn_parent_fthr_bckA_01 [{shapeType=Capsule, length=0.505, radius=0.505, rotation=<rot 0 -90 0>}]
}
constraints [
 {bone1Name=dyn_arm_fthrA_01_l, bone2Name=dyn_parent_arm_fthr_l, swing1LimitDegrees=5, swing2LimitDegrees=5, twistLimitDegrees=3}
 {bone1Name=dyn_arm_fthrA_02_l, bone2Name=dyn_arm_fthrA_01_l, swing1LimitDegrees=10, swing2LimitDegrees=10, twistLimitDegrees=3}
 {
  bone1Name=dyn_hinge_arm_fthr_l
  bone2Name=dyn_parent_arm_fthr_l
  swing1LimitDegrees=45
  swing1Motion=Free
  swing2LimitDegrees=45
  swing2Motion=Free
  twistLimitDegrees=45
  twistMotion=Free
 }
 {bone1Name=dyn_arm_fthrB_01_l, bone2Name=dyn_hinge_arm_fthr_l, swing1LimitDegrees=39, swing2LimitDegrees=30, twistLimitDegrees=3}
 {bone1Name=dyn_arm_fthrB_02_l, bone2Name=dyn_arm_fthrB_01_l, swing1LimitDegrees=10, swing2LimitDegrees=10, twistLimitDegrees=3}
 {bone1Name=dyn_arm_fthrC_01_l, bone2Name=dyn_hinge_arm_fthr_l, swing1LimitDegrees=10, swing2LimitDegrees=20, twistLimitDegrees=3}
 {bone1Name=dyn_arm_fthrC_02_l, bone2Name=dyn_arm_fthrC_01_l, swing1LimitDegrees=10, swing2LimitDegrees=20, twistLimitDegrees=3}
 {bone1Name=dyn_arm_fthrA_01_r, bone2Name=dyn_parent_arm_fthr_r, swing1LimitDegrees=10, swing2LimitDegrees=10, twistLimitDegrees=3}
 {bone1Name=dyn_arm_fthrA_02_r, bone2Name=dyn_arm_fthrA_01_r, swing1LimitDegrees=10, swing2LimitDegrees=10, twistLimitDegrees=3}
 {
  bone1Name=dyn_hinge_arm_fthr_r
  bone2Name=dyn_parent_arm_fthr_r
  swing1LimitDegrees=45
  swing1Motion=Free
  swing2LimitDegrees=45
  swing2Motion=Free
  twistLimitDegrees=45
  twistMotion=Free
 }
 {bone1Name=dyn_arm_fthrB_01_r, bone2Name=dyn_hinge_arm_fthr_r, swing1LimitDegrees=10, swing2LimitDegrees=10, twistLimitDegrees=3}
 {bone1Name=dyn_arm_fthrB_02_r, bone2Name=dyn_arm_fthrB_01_r, swing1LimitDegrees=10, swing2LimitDegrees=10, twistLimitDegrees=3}
 {bone1Name=dyn_arm_fthrC_01_r, bone2Name=dyn_hinge_arm_fthr_r, swing1LimitDegrees=10, swing2LimitDegrees=10, twistLimitDegrees=3}
 {bone1Name=dyn_arm_fthrC_02_r, bone2Name=dyn_arm_fthrC_01_r, swing1LimitDegrees=10, swing2LimitDegrees=10, twistLimitDegrees=3}
 {bone1Name=dyn_fthr_bckA_01, bone2Name=dyn_parent_fthr_bckA_01, swing1LimitDegrees=5, swing2LimitDegrees=5, twistLimitDegrees=5}
 {bone1Name=dyn_fthr_bckA_02, bone2Name=dyn_fthr_bckA_01, swing1LimitDegrees=5, swing2LimitDegrees=5, twistLimitDegrees=5}
]
