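# Physics asset for the side cloth skirt of the male medium Dark Minion player body cosmetic.
# Capsule bodies cover the pelvis, wolf thighs and skirt bones, with left, right and front-left chains hanging from small driver bodies constrained to the pelvis.
# The front-left segment also links to the first left segment, tying those chains together.
# The left chain stiffens toward its tip through tighter limits, while the right keeps uniform limits.
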PhysicsAsset /BRCosmetics/Characters/Player/Male/Medium/Bodies/M_MED_Dark_Minion/Meshes/M_MED_Dark_Minion_Side_Cloth
bodies {
 dyn_driver_skirt_fl [{shapeType=Capsule, length=0.505, radius=0.505, rotation=<rot 0 -90 0>}]
 dyn_driver_skirt_l [{shapeType=Capsule, length=0.505, radius=0.505, rotation=<rot 0 -90 0>}]
 dyn_driver_skirt_r [{shapeType=Capsule, length=0.505, radius=0.505, rotation=<rot 0 -90 0>}]
 dyn_skirt_01_fl [{shapeType=Capsule, length=9.25, radius=6.25, center=<2 5.5 0>, rotation=<rot 90 0 0>}]
 dyn_skirt_01_l [{shapeType=Capsule, length=9.25, radius=11.75, center=<0.1 11.5 0>, rotation=<rot 0 -11.3941 0>}]
 dyn_skirt_01_r [{shapeType=Capsule, length=9, radius=5.5, shapeName=dyn_skirt_01_r_capsule, center=<8 6.5 0>, rotation=<rot 0 90 90>}]
 dyn_skirt_02_l [{shapeType=Capsule, length=6.5, radius=8.5, center=<1 7.8492 0.0041>, rotation=<rot 0.2936 -10.9663 1.5152>}]
 dyn_skirt_02_r [{shapeType=Capsule, length=9, radius=5, shapeName=dyn_skirt_02_r_capsule, center=<8 4.5 0>, rotation=<rot 0 90 90>}]
 dyn_skirt_03_l [{shapeType=Capsule, length=7, radius=7.75, center=<2.0919 6 0>, rotation=<rot 90 -5.0001 0>}]
 dyn_skirt_03_r [{shapeType=Capsule, length=9, radius=5, shapeName=dyn_skirt_03_r_capsule, center=<8 4.5 0>, rotation=<rot 0 90 90>}]
 dyn_skirt_04_l [{shapeType=Capsule, length=7, radius=6, center=<3 5.25 0>}]
 dyn_skirt_04_r [{shapeType=Capsule, length=9, radius=5, shapeName=dyn_skirt_04_r_capsule, center=<8 4.5 0>, rotation=<rot 0 90 90>}]
 pelvis [{shapeType=Capsule, length=7.5, radius=10.5, center=<-6.1772 2.8766 0>}]
 wolf_thigh_l [
  {shapeType=Capsule, length=8.5, radius=12.7826, center=<-0.7663 0.984 4.9939>, rotation=<rot 0.0137 -102.1177 -91.515>}
  {shapeType=Capsule, length=9.5, radius=9.7826, center=<-13.1549 4.4152 2.8634>, rotation=<rot -0.1162 -98.1942 -85.4687>}
 ]
 wolf_thigh_r [{shapeType=Capsule, length=25.7709, radius=11.6256, center=<14.515 -3.3137 -2.2782>, rotation=<rot 0 -100 -87.3521>}]
}
constraints [
 {
  bone1Name=wolf_thigh_l
  bone2Name=pelvis
  swing1LimitDegrees=45
  swing1Motion=Free
  swing2LimitDegrees=45
  swing2Motion=Free
  twistLimitDegrees=45
  twistMotion=Free
 }
 {
  bone1Name=wolf_thigh_r
  bone2Name=pelvis
  swing1LimitDegrees=45
  swing1Motion=Free
  swing2LimitDegrees=45
  swing2Motion=Free
  twistLimitDegrees=45
  twistMotion=Free
 }
 {
  bone1Name=dyn_driver_skirt_l
  bone2Name=pelvis
  swing1LimitDegrees=45
  swing1Motion=Free
  swing2LimitDegrees=45
  swing2Motion=Free
  twistLimitDegrees=45
  twistMotion=Free
 }
 {bone1Name=dyn_skirt_01_l, bone2Name=dyn_driver_skirt_l, swing1LimitDegrees=45, swing2LimitDegrees=45, twistLimitDegrees=15}
 {bone1Name=dyn_skirt_02_l, bone2Name=dyn_skirt_01_l, swing1LimitDegrees=60, swing2LimitDegrees=60, twistLimitDegrees=30}
 {bone1Name=dyn_skirt_03_l, bone2Name=dyn_skirt_02_l, swing1LimitDegrees=15, swing2LimitDegrees=15, twistLimitDegrees=3}
 {bone1Name=dyn_skirt_04_l, bone2Name=dyn_skirt_03_l, swing1LimitDegrees=10, swing2LimitDegrees=10, twistLimitDegrees=3}
 {bone1Name=dyn_driver_skirt_r, bone2Name=pelvis, swing1LimitDegrees=45, swing2LimitDegrees=45, twistLimitDegrees=45}
 {
  bone1Name=dyn_driver_skirt_fl
  bone2Name=pelvis
  swing1LimitDegrees=45
  swing1Motion=Free
  swing2LimitDegrees=45
  swing2Motion=Free
  twistLimitDegrees=45
  twistMotion=Free
 }
 {bone1Name=dyn_skirt_01_fl, bone2Name=dyn_driver_skirt_fl, swing1LimitDegrees=45, swing2LimitDegrees=45, twistLimitDegrees=15}
 {bone1Name=dyn_skirt_01_l, bone2Name=dyn_skirt_01_fl, swing1LimitDegrees=45, swing2LimitDegrees=45, twistLimitDegrees=45}
 {bone1Name=dyn_skirt_01_r, bone2Name=dyn_driver_skirt_r, swing1LimitDegrees=45, swing2LimitDegrees=45, twistLimitDegrees=45}
 {bone1Name=dyn_skirt_02_r, bone2Name=dyn_skirt_01_r, swing1LimitDegrees=45, swing2LimitDegrees=45, twistLimitDegrees=45}
 {bone1Name=dyn_skirt_03_r, bone2Name=dyn_skirt_02_r, swing1LimitDegrees=45, swing2LimitDegrees=45, twistLimitDegrees=45}
 {bone1Name=dyn_skirt_04_r, bone2Name=dyn_skirt_03_r, swing1LimitDegrees=45, swing2LimitDegrees=45, twistLimitDegrees=45}
]
