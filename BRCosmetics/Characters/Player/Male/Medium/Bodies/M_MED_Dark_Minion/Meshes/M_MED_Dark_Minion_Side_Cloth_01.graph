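# Physics asset for the Dark Minion male medium body's side cloth, defining capsule bodies and joints for skirt bone dynamics.
# Pelvis and wolf-thigh capsules provide leg and hip volumes, while tiny driver and parent capsules anchor the skirt to the pelvis.
# A five-link front chain and two single-bone flaps (fl, fAl) hang from these anchors with angular limits.
# Pelvis joints are angularly free, and the left driver carries no skirt bone.
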PhysicsAsset /BRCosmetics/Characters/Player/Male/Medium/Bodies/M_MED_Dark_Minion/Meshes/M_MED_Dark_Minion_Side_Cloth_01
bodies {
 dyn_driver_skirt_fAl [{shapeType=Capsule, length=1, radius=1, center=<0.1 0 0>}]
 dyn_driver_skirt_fl [{shapeType=Capsule, length=1, radius=1, center=<0.1 0 0>}]
 dyn_driver_skirt_frt [{shapeType=Capsule, length=0.505, radius=0.505, rotation=<rot 0 -90 0>}]
 dyn_driver_skirt_l [{shapeType=Capsule, length=0.505, radius=0.505, rotation=<rot 0 -90 0>}]
 dyn_parent_skirt_fAl [{shapeType=Capsule, length=0.505, radius=0.505, rotation=<rot 0 -90 0>}]
 dyn_parent_skirt_fl [{shapeType=Capsule, length=0.505, radius=0.505, rotation=<rot 0 -90 0>}]
 dyn_skirt_01_fAl [{shapeType=Capsule, length=8.5, radius=4.5, center=<3 4.9595 0>, rotation=<rot 90 1.7899 4.8547>}]
 dyn_skirt_01_fl [
  {shapeType=Capsule, length=9.25, radius=7.5, center=<4.0562 7.3323 -2.2878>, rotation=<rot -3.2426 -19.4684 8.1382>}
  {shapeType=Capsule, length=9, radius=4.5, center=<13.021 4.707 -3.2604>, rotation=<rot 2.2272 -19.8966 16.4732>}
 ]
 dyn_skirt_frt_01 [{shapeType=Capsule, length=8, radius=7.75, center=<3 -7.5 0>}]
 dyn_skirt_frt_02 [{shapeType=Capsule, length=7, radius=7.25, center=<2.3062 -8.5579 0.2295>, rotation=<rot 0 0 17.8485>}]
 dyn_skirt_frt_03 [{shapeType=Capsule, length=6.25, radius=6.25, center=<3 -7.5 0>, rotation=<rot 0 0 14.8647>}]
 dyn_skirt_frt_04 [{shapeType=Capsule, length=4.25, radius=4.75, center=<4 -6.1 0>, rotation=<rot 0 0 11.2068>}]
 dyn_skirt_frt_05 [{shapeType=Capsule, length=2.25, radius=3.75, center=<5 -5.25 0>, rotation=<rot 90 0 0>}]
 pelvis [{shapeType=Capsule, length=7.5, radius=10.5, center=<-6.1772 2.8766 0>}]
 wolf_thigh_l [
  {shapeType=Capsule, length=10, radius=11.2826, center=<-0.9508 0.1249 4.9941>, rotation=<rot 0.0137 -102.1178 -91.515>}
  {shapeType=Capsule, length=34, radius=9.5326, center=<-9.5947 4.4262 2.2485>, rotation=<rot 0.1392 -101.9068 -89.4252>}
 ]
 wolf_thigh_r [{shapeType=Capsule, length=28.1139, radius=9.2826, center=<14.3533 -4.3761 -1.7344>, rotation=<rot 0 -100 -87.3521>}]
}
constraints [
 {
  bone1Name=wolf_thigh_l
  bone2Name=pelvis
  swing1LimitDegrees=45
  swing1Motion=Free
  swing2LimitDegrees=45
  swing2Motion=Free
  twistLimitDegrees=45
  twistMotion=Free
 }
 {
  bone1Name=wolf_thigh_r
  bone2Name=pelvis
  swing1LimitDegrees=45
  swing1Motion=Free
  swing2LimitDegrees=45
  swing2Motion=Free
  twistLimitDegrees=45
  twistMotion=Free
 }
 {
  bone1Name=dyn_driver_skirt_l
  bone2Name=pelvis
  swing1LimitDegrees=45
  swing1Motion=Free
  swing2LimitDegrees=45
  swing2Motion=Free
  twistLimitDegrees=45
  twistMotion=Free
 }
 {bone1Name=dyn_driver_skirt_fl, bone2Name=dyn_parent_skirt_fl, swing1LimitDegrees=45, swing2LimitDegrees=45, twistLimitDegrees=45}
 {bone1Name=dyn_skirt_01_fl, bone2Name=dyn_driver_skirt_fl, swing1LimitDegrees=60, swing2LimitDegrees=60, twistLimitDegrees=8}
 {
  bone1Name=dyn_driver_skirt_frt
  bone2Name=pelvis
  swing1LimitDegrees=45
  swing1Motion=Free
  swing2LimitDegrees=45
  swing2Motion=Free
  twistLimitDegrees=45
  twistMotion=Free
 }
 {bone1Name=dyn_skirt_frt_01, bone2Name=dyn_driver_skirt_frt, swing1LimitDegrees=45, swing2LimitDegrees=15, twistLimitDegrees=15}
 {bone1Name=dyn_skirt_frt_02, bone2Name=dyn_skirt_frt_01, swing1LimitDegrees=45, swing2LimitDegrees=45, twistLimitDegrees=15}
 {bone1Name=dyn_skirt_frt_03, bone2Name=dyn_skirt_frt_02, swing1LimitDegrees=45, swing2LimitDegrees=45, twistLimitDegrees=15}
 {bone1Name=dyn_skirt_frt_04, bone2Name=dyn_skirt_frt_03, swing1LimitDegrees=45, swing2LimitDegrees=45, twistLimitDegrees=15}
 {bone1Name=dyn_skirt_frt_05, bone2Name=dyn_skirt_frt_04, swing1LimitDegrees=45, swing2LimitDegrees=45, twistLimitDegrees=15}
 {bone1Name=dyn_driver_skirt_fAl, bone2Name=dyn_parent_skirt_fAl, swing1LimitDegrees=45, swing2LimitDegrees=45, twistLimitDegrees=45}
 {bone1Name=dyn_skirt_01_fAl, bone2Name=dyn_driver_skirt_fAl, swing1LimitDegrees=60, swing2LimitDegrees=60, twistLimitDegrees=8}
 {
  bone1Name=dyn_parent_skirt_fAl
  bone2Name=pelvis
  swing1LimitDegrees=45
  swing1Motion=Free
  swing2LimitDegrees=45
  swing2Motion=Free
  twistLimitDegrees=45
  twistMotion=Free
 }
 {
  bone1Name=dyn_parent_skirt_fl
  bone2Name=pelvis
  swing1LimitDegrees=45
  swing1Motion=Free
  swing2LimitDegrees=45
  swing2Motion=Free
  twistLimitDegrees=45
  twistMotion=Free
 }
]
