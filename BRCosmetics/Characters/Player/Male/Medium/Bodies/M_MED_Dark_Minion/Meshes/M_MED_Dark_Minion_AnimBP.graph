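# Animation Blueprint for the Dark Minion male body cosmetic, copying the attached parent's pose and adding secondary motion.
# Upper-arm constraints feed a crouch-driven Control Rig, then RigidBody feather physics, additive belt and skirt correctives weighted by inherited leg and crouch floats, cloth RigidBody passes and thigh look-ats.
# The event graph toggles the LegAdjust morph by ExtraPart mesh presence.
# Its computed gravity override vector only feeds a debug draw.
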AnimBlueprint /BRCosmetics/Characters/Player/Male/Medium/Bodies/M_MED_Dark_Minion/Meshes/M_MED_Dark_Minion_AnimBP
target_skeleton=&/M_MED_Dark_Minion_Skeleton
graphs [
 AnimGraph {nodes [
   "Transform (Modify) Bone - Bone: dyn_thigh_cor_r"(Alpha=0, Translation="0,0,0", Rotation="0,0,0", Scale="1,1,1") {
    rotationMode=BMM_Additive
    rotationSpace=BCS_BoneSpace
    translationMode=BMM_Additive
    translationSpace=BCS_BoneSpace
    alphaBoolBlend {blendOption=Linear}
    alphaScaleBiasClamp {bInterpResult=true, interpSpeedDecreasing=20, interpSpeedIncreasing=20}
    boneToModify {boneName=dyn_thigh_cor_r}
   }
   n1 = Get SubLegHighL() -> (SubLegHighL:float)
   n2 = Get LegUpClothR() -> (LegUpClothR:float)
   n3 = "Constraint - con_upperarm_l (PPPP)"!(
    ComponentPose=Local To Component(LocalPose=Copy Pose From Mesh() {bUseAttachedParent=true})
    Alpha=1
    ConstraintWeights_0=0.1245
    ConstraintWeights_1:float=0.1267
    ConstraintWeights_2:float=0.1603
    ConstraintWeights_3:float=0.5885
   ) -> (Pose) {
    alphaBoolBlend {blendOption=Linear}
    boneToModify {boneName=con_upperarm_l}
    constraintSetup [{targetBone {boneName=spine_03}}, {targetBone {boneName=spine_04}}, {targetBone {boneName=spine_05}}, {targetBone {boneName=clavicle_l}}]
    constraintWeights [1,1,1,1]
   }
   n4 = "Constraint - con_upperarm_r (PPPP)"!(
    ComponentPose=@n3
    Alpha=1
    ConstraintWeights_0=0.5842
    ConstraintWeights_1:float=0.1304
    ConstraintWeights_2:float=0.1288
    ConstraintWeights_3:float=0.1566
   ) -> (Pose) {
    alphaBoolBlend {blendOption=Linear}
    boneToModify {boneName=con_upperarm_r}
    constraintSetup [{targetBone {boneName=clavicle_r}}, {targetBone {boneName=spine_03}}, {targetBone {boneName=spine_04}}, {targetBone {boneName=spine_05}}]
    constraintWeights [1,1,1,1]
   }
   n5 = Control Rig!(Source=Component To Local(ComponentPose=@n4), Alpha, crouch_alpha:double=Get CrouchFloat() -> (CrouchFloat:float)) -> (Pose) {
    alphaBoolBlend {blendOption=Linear}
    controlRigAssetReference {blueprintRigClass=&/M_MED_Dark_Minion_CtrlRig.M_MED_Dark_Minion_CtrlRig_C}
   }
   n6 = RigidBody(ComponentPose=Local To Component(LocalPose=@n5), Alpha=0.7) {
    bOverrideWorldGravity=true
    overridePhysicsAsset=&/M_MED_Dark_Minion_Fthrs_Physics
    simulationSpace=BaseBoneSpace
    alphaBoolBlend {blendOption=Linear}
    baseBoneRef {boneName=root}
    componentAppliedLinearAccClamp=<3000 3000 3000>
    componentLinearAccScale=<0.2 0.2 0.1>
    componentLinearVelScale=<0.2 0.2 0.1>
    overrideWorldGravity {z=-15}
   }
   n7 = "Transform (Modify) Bone - Bone: dyn_beltA_01_r"(ComponentPose=@n6, Alpha=1, Translation="0,-1,0", Rotation="0,-10,0", Scale="1,1,1") {
    rotationMode=BMM_Additive
    rotationSpace=BCS_BoneSpace
    translationMode=BMM_Additive
    translationSpace=BCS_BoneSpace
    alphaBoolBlend {blendOption=Linear}
    alphaScaleBiasClamp {bInterpResult=true, interpSpeedDecreasing=18, interpSpeedIncreasing=18}
    boneToModify {boneName=dyn_beltA_01_r}
    rotation {yaw=-10}
    translation {y=-1}
   }
   n8 = "Transform (Modify) Bone - Bone: pelvis_offset"(ComponentPose=@n7, Alpha=Get CrouchFloat(), Translation="4,0,0", Rotation="0,0,0", Scale="1,1,1") {
    translationMode=BMM_Additive
    translationSpace=BCS_BoneSpace
    alphaBoolBlend {blendOption=Linear}
    alphaScaleBiasClamp {bInterpResult=true, interpSpeedDecreasing=18, interpSpeedIncreasing=18}
    boneToModify {boneName=pelvis_offset}
    translation {x=4}
   }
   n9 = "Transform (Modify) Bone - Bone: dyn_beltA_01_l"(ComponentPose=@n8, Alpha=Get LegUpStrapL() -> (LegUpStrapL:float), Translation="-2,-3,-1.947", Rotation="8.951,0.646,-13.840", Scale="1,1,1") {
    rotationMode=BMM_Additive
    rotationSpace=BCS_BoneSpace
    translationMode=BMM_Additive
    translationSpace=BCS_BoneSpace
    alphaBoolBlend {blendOption=Linear}
    alphaScaleBiasClamp {bInterpResult=true, interpSpeedDecreasing=18, interpSpeedIncreasing=18}
    boneToModify {boneName=dyn_beltA_01_l}
    rotation=<rot 8.951 0.646 -13.8401>
    translation=<-2 -3 -1.947>
   }
   n10 = "Transform (Modify) Bone - Bone: dyn_beltA_01_r"(ComponentPose=@n9, Alpha=Get LegUpStrapR() -> (LegUpStrapR:float), Translation="0,0,0", Rotation="0,-10,0", Scale="1,1,1") {
    rotationMode=BMM_Additive
    rotationSpace=BCS_BoneSpace
    translationMode=BMM_Additive
    translationSpace=BCS_BoneSpace
    alphaBoolBlend {blendOption=Linear}
    alphaScaleBiasClamp {bInterpResult=true, interpSpeedDecreasing=18, interpSpeedIncreasing=18}
    boneToModify {boneName=dyn_beltA_01_r}
    rotation {yaw=-10}
   }
   n11 = "Transform (Modify) Bone - Bone: dyn_parent_skirt_fr"(ComponentPose=@n10, Alpha=1, Translation="0.000,1,0", Rotation="0,0,0", Scale="1,1,1") {
    rotationSpace=BCS_BoneSpace
    translationMode=BMM_Additive
    translationSpace=BCS_BoneSpace
    alphaBoolBlend {blendOption=Linear}
    alphaScaleBiasClamp {bInterpResult=true, interpSpeedDecreasing=20, interpSpeedIncreasing=20}
    boneToModify {boneName=dyn_parent_skirt_fr}
    translation {y=1}
   }
   n12 = "Transform (Modify) Bone - Bone: dyn_parent_skirt_frA"(ComponentPose=@n11, Alpha=1, Translation="0.000,2,0", Rotation="0,0,0", Scale="1,1,1") {
    rotationSpace=BCS_BoneSpace
    translationMode=BMM_Additive
    translationSpace=BCS_BoneSpace
    alphaBoolBlend {blendOption=Linear}
    alphaScaleBiasClamp {bInterpResult=true, interpSpeedDecreasing=20, interpSpeedIncreasing=20}
    boneToModify {boneName=dyn_parent_skirt_frA}
    translation {y=2}
   }
   n13 = "Transform (Modify) Bone - Bone: dyn_parent_skirt_fr"(ComponentPose=@n12, Alpha=@n2, Translation="0.000,2,0", Rotation="-1.5,30,-3", Scale="1,1,1") {
    rotationMode=BMM_Additive
    rotationSpace=BCS_BoneSpace
    translationMode=BMM_Additive
    translationSpace=BCS_BoneSpace
    alphaBoolBlend {blendOption=Linear}
    alphaScaleBiasClamp {bInterpResult=true, interpSpeedDecreasing=20, interpSpeedIncreasing=20}
    boneToModify {boneName=dyn_parent_skirt_fr}
    rotation=<rot -1.5 30 -3>
    translation {y=2}
   }
   n14 = "Transform (Modify) Bone - Bone: dyn_parent_skirt_frA"(ComponentPose=@n13, Alpha=@n2, Translation="0.000,1,0", Rotation="8.029,6.946,0.344", Scale="1,1,1") {
    rotationMode=BMM_Additive
    rotationSpace=BCS_BoneSpace
    translationMode=BMM_Additive
    translationSpace=BCS_BoneSpace
    alphaBoolBlend {blendOption=Linear}
    alphaScaleBiasClamp {bInterpResult=true, interpSpeedDecreasing=20, interpSpeedIncreasing=20}
    boneToModify {boneName=dyn_parent_skirt_frA}
    rotation=<rot 8.029 6.946 0.344>
    translation {y=1}
   }
   n15 = "Transform (Modify) Bone - Bone: dyn_parent_skirt_frA"(ComponentPose=@n14, Alpha=Get LegOutR() -> (LegOutR:float), Translation="0.000,2,0", Rotation="0,-10,0", Scale="1,1,1") {
    rotationMode=BMM_Additive
    rotationSpace=BCS_BoneSpace
    translationMode=BMM_Additive
    translationSpace=BCS_BoneSpace
    alphaBoolBlend {blendOption=Linear}
    alphaScaleBiasClamp {bInterpResult=true, interpSpeedDecreasing=20, interpSpeedIncreasing=20}
    boneToModify {boneName=dyn_parent_skirt_frA}
    rotation {yaw=-10}
    translation {y=2}
   }
   n16 = "Transform (Modify) Bone - Bone: dyn_beltA_01_r"(ComponentPose=@n15, Alpha=@n2, Translation="0,-1,0", Rotation="0,-5,0", Scale="1,1,1") {
    rotationMode=BMM_Additive
    rotationSpace=BCS_BoneSpace
    translationMode=BMM_Additive
    translationSpace=BCS_BoneSpace
    alphaBoolBlend {blendOption=Linear}
    alphaScaleBiasClamp {bInterpResult=true, interpSpeedDecreasing=18, interpSpeedIncreasing=18}
    boneToModify {boneName=dyn_beltA_01_r}
    rotation {yaw=-5}
    translation {y=-1}
   }
   n17 = "Transform (Modify) Bone - Bone: dyn_parent_skirt_l"(ComponentPose=@n16, Alpha=Get LegUpL() -> (LegUpL:float), Translation="-2,0.5,5", Rotation="0,0,-30", Scale="1,1,1") {
    rotationMode=BMM_Additive
    rotationSpace=BCS_BoneSpace
    translationMode=BMM_Additive
    translationSpace=BCS_BoneSpace
    alphaBoolBlend {blendOption=Linear}
    boneToModify {boneName=dyn_parent_skirt_l}
    rotation {roll=-30}
    translation=<-2 0.5 5>
   }
   n18 = "Transform (Modify) Bone - Bone: dyn_parent_skirt_l"(ComponentPose=@n17, Alpha=Get LegOutClothL() -> (LegOutClothL:float), Translation="0,1,0", Rotation="0,0,0", Scale="1,1,1") {
    rotationMode=BMM_Additive
    rotationSpace=BCS_BoneSpace
    translationMode=BMM_Additive
    translationSpace=BCS_BoneSpace
    alphaBoolBlend {blendOption=Linear}
    boneToModify {boneName=dyn_parent_skirt_l}
    translation {y=1}
   }
   n19 = Get LegUpClothL() -> (LegUpClothL:float)
   n20 = "Transform (Modify) Bone - Bone: dyn_parent_skirt_fl"(ComponentPose=@n18, Alpha=@n19, Translation="-5,1,0", Rotation="-20,0,0", Scale="1,1,1") {
    rotationMode=BMM_Additive
    rotationSpace=BCS_BoneSpace
    translationMode=BMM_Additive
    translationSpace=BCS_BoneSpace
    alphaBoolBlend {blendOption=Linear}
    alphaScaleBiasClamp {bInterpResult=true, interpSpeedDecreasing=20, interpSpeedIncreasing=20}
    boneToModify {boneName=dyn_parent_skirt_fl}
    rotation {pitch=-20}
    translation=<-5 1>
   }
   n21 = "Transform (Modify) Bone - Bone: dyn_thigh_cor"(ComponentPose=@n20, Alpha=@n19, Translation="-5,5,0", Rotation="0,0,0", Scale="1,1,1") {
    rotationMode=BMM_Additive
    rotationSpace=BCS_BoneSpace
    translationMode=BMM_Additive
    translationSpace=BCS_BoneSpace
    alphaBoolBlend {blendOption=Linear}
    alphaScaleBiasClamp {bInterpResult=true, interpSpeedDecreasing=20, interpSpeedIncreasing=20}
    boneToModify {boneName=dyn_thigh_cor}
    translation=<-5 5>
   }
   n22 = "Transform (Modify) Bone - Bone: dyn_parent_skirt_fAl"(ComponentPose=@n21, Alpha=Get LegUpClothL(), Translation="-1,-0.5,0", Rotation="0,35,0", Scale="1,1,1") {
    rotationMode=BMM_Additive
    rotationSpace=BCS_BoneSpace
    translationMode=BMM_Additive
    translationSpace=BCS_BoneSpace
    alphaBoolBlend {blendOption=Linear}
    alphaScaleBiasClamp {bInterpResult=true, interpSpeedDecreasing=20, interpSpeedIncreasing=20}
    boneToModify {boneName=dyn_parent_skirt_fAl}
    rotation {yaw=35}
    translation=<-1 -0.5>
   }
   n23 = Get SubLegL() -> (SubLegL:float)
   n24 = "Transform (Modify) Bone - Bone: dyn_parent_skirt_fl"(ComponentPose=@n22, Alpha=@n23, Translation="-6,1,-1", Rotation="-15,0,0", Scale="0.7,0.7,0.7") {
    rotationMode=BMM_Additive
    rotationSpace=BCS_BoneSpace
    scaleMode=BMM_Additive
    scaleSpace=BCS_BoneSpace
    translationMode=BMM_Additive
    translationSpace=BCS_BoneSpace
    alphaBoolBlend {blendOption=Linear}
    alphaScaleBiasClamp {bInterpResult=true}
    boneToModify {boneName=dyn_parent_skirt_fl}
    rotation {pitch=-15}
    scale=<0.7 0.7 0.7>
    translation=<-6 1 -1>
   }
   n25 = "Transform (Modify) Bone - Bone: dyn_parent_skirt_fAl"(ComponentPose=@n24, Alpha=@n23, Translation="-9,-2,0", Rotation="0,30,0", Scale="0.7,0.7,0.7") {
    rotationMode=BMM_Additive
    rotationSpace=BCS_BoneSpace
    scaleMode=BMM_Additive
    scaleSpace=BCS_BoneSpace
    translationMode=BMM_Additive
    translationSpace=BCS_BoneSpace
    alphaBoolBlend {blendOption=Linear}
    alphaScaleBiasClamp {bInterpResult=true}
    boneToModify {boneName=dyn_parent_skirt_fAl}
    rotation {yaw=30}
    scale=<0.7 0.7 0.7>
    translation=<-9 -2>
   }
   n26 = "Transform (Modify) Bone - Bone: dyn_parent_skirt_fl"(ComponentPose=@n25, Alpha=@n1, Translation="-6,2,-4", Rotation="0,0,0", Scale="1,1,1") {
    rotationMode=BMM_Additive
    rotationSpace=BCS_BoneSpace
    scaleMode=BMM_Additive
    scaleSpace=BCS_BoneSpace
    translationMode=BMM_Additive
    translationSpace=BCS_BoneSpace
    alphaBoolBlend {blendOption=Linear}
    alphaScaleBiasClamp {bInterpResult=true}
    boneToModify {boneName=dyn_parent_skirt_fl}
    translation=<-6 2 -4>
   }
   n27 = "Transform (Modify) Bone - Bone: dyn_parent_skirt_fAl"(ComponentPose=@n26, Alpha=@n1, Translation="-3,3,0", Rotation="0,0,0", Scale="1,1,1") {
    rotationMode=BMM_Additive
    rotationSpace=BCS_BoneSpace
    scaleMode=BMM_Additive
    scaleSpace=BCS_BoneSpace
    translationMode=BMM_Additive
    translationSpace=BCS_BoneSpace
    alphaBoolBlend {blendOption=Linear}
    alphaScaleBiasClamp {bInterpResult=true}
    boneToModify {boneName=dyn_parent_skirt_fAl}
    translation=<-3 3>
   }
   n28 = "Transform (Modify) Bone - Bone: dyn_thigh_cor"(ComponentPose=@n27, Alpha=@n1, Translation="-3,-5,-3", Rotation="0,0,0", Scale="1,1,1") {
    rotationMode=BMM_Additive
    rotationSpace=BCS_BoneSpace
    translationMode=BMM_Additive
    translationSpace=BCS_BoneSpace
    alphaBoolBlend {blendOption=Linear}
    alphaScaleBiasClamp {bInterpResult=true, interpSpeedDecreasing=20, interpSpeedIncreasing=20}
    boneToModify {boneName=dyn_thigh_cor}
    translation=<-3 -5 -3>
   }
   n29 = RigidBody!(ComponentPose=@n28, Alpha, OverrideWorldGravity:Vector="0,0,-980", ExternalForce) -> (Pose) {
    bOverrideWorldGravity=true
    overridePhysicsAsset=&/M_MED_Dark_Minion_Bck_Skirt
    simulationSpace=BaseBoneSpace
    alphaBoolBlend {blendOption=Linear}
    baseBoneRef {boneName=root}
    componentAppliedLinearAccClamp=<1100 1100 1000>
    componentLinearAccScale=<0.3 0.3 0.3>
    componentLinearVelScale=<0.3 0.3 0.3>
    overrideWorldGravity {z=-980}
   }
   n30 = RigidBody!(ComponentPose=@n29, Alpha, ExternalForce) -> (Pose) {
    bOverrideWorldGravity=true
    overridePhysicsAsset=&/M_MED_Dark_Minion_Side_Cloth
    simulationSpace=BaseBoneSpace
    alphaBoolBlend {blendOption=Linear}
    baseBoneRef {boneName=root}
    componentAppliedLinearAccClamp=<0 0 0>
    componentLinearAccScale=<0.3 0.3 0.3>
    componentLinearVelScale=<0.3 0.3 0.3>
    overrideWorldGravity {z=-980}
    simSpaceSettings {maxAngularAcceleration=3000, maxAngularVelocity=3000, maxLinearAcceleration=3000, maxLinearVelocity=3000, worldAlpha=1}
   }
   n31 = RigidBody!(ComponentPose=@n30, Alpha, OverrideWorldGravity:Vector="0,0,-980", ExternalForce) -> (Pose) {
    bOverrideWorldGravity=true
    overridePhysicsAsset=&/M_MED_Dark_Minion_Side_Cloth_01
    simulationSpace=BaseBoneSpace
    alphaBoolBlend {blendOption=Linear}
    baseBoneRef {boneName=root}
    componentAppliedLinearAccClamp=<1100 1100 1000>
    componentLinearAccScale=<0.3 0.3 0.3>
    componentLinearVelScale=<0.3 0.3 0.3>
    overrideWorldGravity {z=-980}
   }
   n32 = Look At - Bone: wolf_thigh_l(ComponentPose=@n31, Alpha=1) {
    alphaBoolBlend {blendOption=Linear}
    boneToModify {boneName=wolf_thigh_l}
    lookAtLocation {x=0}
    lookAtTarget {boneReference {boneName=calf_l}}
    lookAt_Axis {axis=<-1 0>}
   }
   n33 = Look At - Bone: wolf_thigh_r(ComponentPose=@n32, Alpha=1) {
    alphaBoolBlend {blendOption=Linear}
    boneToModify {boneName=wolf_thigh_r}
    lookAtLocation {x=0}
    lookAtTarget {boneReference {boneName=calf_r}}
    lookAt_Axis {axis=<1 0>}
   }
   n34 = "Transform (Modify) Bone - Bone: wolf_thigh_l"(ComponentPose=@n33, Alpha=1, Translation="0,0,0", Rotation="3,10,0", Scale="1,1,1") {
    rotationMode=BMM_Additive
    rotationSpace=BCS_BoneSpace
    alphaBoolBlend {blendOption=Linear}
    boneToModify {boneName=wolf_thigh_l}
    rotation {pitch=3, yaw=10}
   }
   n35 = "Transform (Modify) Bone - Bone: wolf_thigh_r"(ComponentPose=@n34, Alpha=1, Translation="0,0,0", Rotation="3,10,0", Scale="1,1,1") {
    rotationMode=BMM_Additive
    rotationSpace=BCS_BoneSpace
    alphaBoolBlend {blendOption=Linear}
    boneToModify {boneName=wolf_thigh_r}
    rotation {pitch=3, yaw=10}
   }
   Output Pose(Result=Component To Local(ComponentPose=@n35)) {layerGroup=None}
  ]}
 EventGraph {nodes [
   Event OnGravityOverrideUpdate() {is=title K2Node_Event}
   > n1 = Gravity Override State Function DEPRECATED(CharacterPartAnimInstance=Self-Reference(), GameTimeInSeconds=GetGameTimeInSeconds(), GravityOverrideParams=Get Gravity Override Params())
   > n2 = Set gravityVector(gravityVector:Vector=@n1) -> (Output_Get:Vector)
   > DrawVectorFromJointFunction!(
    execute
    SkelMeshComponent:Scene Component Object Reference
    JointName:Name=head
    LineEnd:Vector=@n2
    JointOffset:Vector="0, 0, 0"
    LineColor:Linear Color Structure=(R=1,G=0,B=0,A=1)
    Thickness:double=2
    Foreground:EDrawDebugSceneDepthPriorityGroup Enum=Foreground
   ) -> (then)
   Event BlueprintUpdateAnimation()
   > n3 = Get!(execute:Exec) -> (then:Exec, else:Exec, ShoeMesh:Skeletal Mesh Component Budgeted Object Reference)
   > n4 = Is Valid(InputObject=Get SkeletalMesh!(self:Skinned Mesh Component Object Reference=@n3) -> (SkeletalMesh:Skeletal Mesh Object Reference))
   SetMorphTarget!(execute=[@n4.Is Not Valid, @n3.else], self:Anim Instance Object Reference, MorphTargetName:Name=LegAdjust, Value:float=0) -> (then)
   SetMorphTarget(execute=@n4.Is Valid, MorphTargetName=LegAdjust, Value=1)
   n5 = Event OnCharacterCustomizationLoaded()
   > Set ShoeMesh(ShoeMesh:Skeletal Mesh Component Budgeted Object Reference=Get Array Item with EFort Custom Part Type(Array=@n5.CustomMeshParts, PartType=ExtraPart)) -> (Output_Get:Skeletal Mesh Component Budgeted Object Reference)
  ]}
]
variables [
 Body_Mesh {type=SkeletalMeshComponent}
 Gravity Override Params {
  type=GravityOverrideParamsStruct
  default {skydive_DivingDown {gravityOverride=<0 0 1800>}, skydive_Parachuting {gravityOverride=<0 0 -980>}}
 }
 gravityVector {type=Vector}
 ShoeMesh {type=SkeletalMeshComponentBudgeted}
]
