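# Physics asset for the cowl part of the Dark Minion male medium body, giving its cloth and flap bones capsule bodies and joint limits.
# Large capsules cover the spine and clavicles, while small dyn_parent anchor bodies link the cloth chains to spine_04 and spine_05 with free angular motion.
# Three asymmetric back chains and two front flap chains have tight twist limits, with swing generally loosening toward the tips, except on the right back chain.
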PhysicsAsset /BRCosmetics/Characters/Player/Male/Medium/Bodies/M_MED_Dark_Minion/Meshes/Parts/M_MED_Dark_Minion_Cowl_Physics
bodies {
 clavicle_l [{shapeType=Capsule, length=16.25, radius=8.75, center=<2.0936 0 1.7142>, rotation=<rot 90 0 0>}]
 clavicle_r [{shapeType=Capsule, length=16.25, radius=8.75, center=<-1.8522 0 -2.4465>, rotation=<rot 90 0 0>}]
 dyn_clth_bck_01 [{shapeType=Capsule, length=9.25, radius=6.5, center=<0.5 6.5 0>}]
 dyn_clth_bck_01_l [{shapeType=Capsule, length=6.75, radius=6.25, center=<1 6.25 0>}]
 dyn_clth_bck_01_r [{shapeType=Capsule, length=6.75, radius=6.25, center=<1 6.85 0>}]
 dyn_clth_bck_02 [{shapeType=Capsule, length=8, radius=6, center=<2 5.5 0>}]
 dyn_clth_bck_02_l [{shapeType=Capsule, length=5.75, radius=5.25, center=<1.5 5.25 -0.0002>}]
 dyn_clth_bck_02_r [{shapeType=Capsule, length=5.75, radius=5.25, center=<2 5.75 -0.0002>}]
 dyn_clth_bck_03 [{shapeType=Capsule, length=6.5, radius=4.25, center=<4 4.15 0>}]
 dyn_clth_bck_03_l [{shapeType=Capsule, length=4.25, radius=4.5, center=<3 4.5 0>}]
 dyn_clth_bck_03_r [{shapeType=Capsule, length=4.25, radius=4.5, center=<2.5 4.5 0>}]
 dyn_clth_bck_04 [{shapeType=Capsule, length=5.25, radius=3.75, center=<6 3 0>, rotation=<rot 90 0 0>}]
 dyn_clth_bck_04_l [{shapeType=Capsule, length=3.75, radius=3.75, center=<4 4.5 0>}]
 dyn_clth_bck_05_l [{shapeType=Capsule, length=2.25, radius=3.25, center=<5 3.25 0>}]
 dyn_flap_frt_01_l [{shapeType=Capsule, length=3, radius=4.25, center=<3 4.25 0>, rotation=<rot 90 0 0>}]
 dyn_flap_frt_01_r [{shapeType=Capsule, length=3, radius=4.3, center=<1 4.5695 0>, rotation=<rot 90 0 0>}]
 dyn_flap_frt_02_l [{shapeType=Capsule, length=3.5, radius=3, center=<4 3 0>, rotation=<rot 90 0 0>}]
 dyn_flap_frt_02_r [{shapeType=Capsule, length=3.5, radius=4, center=<4 5.25 0.0001>}]
 dyn_parent_clth_bck_01 [{shapeType=Capsule, length=1, radius=1}]
 dyn_parent_clth_bck_01_l [{shapeType=Capsule, length=1, radius=1}]
 dyn_parent_clth_bck_01_r [{shapeType=Capsule, length=1, radius=1}]
 dyn_parent_flap_frt_01_l [{shapeType=Capsule, length=1, radius=1}]
 dyn_parent_flap_frt_01_r [{shapeType=Capsule, length=1, radius=1}]
 spine_01 [{shapeType=Capsule, length=8.7924, radius=16.5175, center=<9.1591 5.4469 0>, rotation=<rot 90 179.1048 177.4058>}]
 spine_03 [{shapeType=Capsule, length=9.5, radius=18, center=<13.3348 7.3021 0>, rotation=<rot 90 0 -8.9572>}]
 spine_04 [{shapeType=Capsule, length=13.25, radius=15.5, center=<6.6263 -1.5672 0>}]
 spine_05 [{shapeType=Capsule, length=17.0654, radius=15.8908, center=<1.33 6.8721 0.0001>}]
}
constraints [
 {bone1Name=spine_03, bone2Name=spine_01, swing1LimitDegrees=45, swing2LimitDegrees=45, twistLimitDegrees=45}
 {
  bone1Name=dyn_parent_clth_bck_01
  bone2Name=spine_04
  swing1LimitDegrees=45
  swing1Motion=Free
  swing2LimitDegrees=45
  swing2Motion=Free
  twistLimitDegrees=45
  twistMotion=Free
 }
 {bone1Name=dyn_clth_bck_01, bone2Name=dyn_parent_clth_bck_01, swing1LimitDegrees=5, swing2LimitDegrees=5, twistLimitDegrees=5}
 {bone1Name=dyn_clth_bck_02, bone2Name=dyn_clth_bck_01, swing1LimitDegrees=8, swing2LimitDegrees=30, twistLimitDegrees=5}
 {bone1Name=dyn_clth_bck_03, bone2Name=dyn_clth_bck_02, swing1LimitDegrees=15, swing2LimitDegrees=45, twistLimitDegrees=5}
 {bone1Name=dyn_clth_bck_04, bone2Name=dyn_clth_bck_03, swing1LimitDegrees=25, swing2LimitDegrees=65, twistLimitDegrees=5}
 {
  bone1Name=dyn_parent_clth_bck_01_l
  bone2Name=spine_05
  swing1LimitDegrees=45
  swing1Motion=Free
  swing2LimitDegrees=45
  swing2Motion=Free
  twistLimitDegrees=45
  twistMotion=Free
 }
 {bone1Name=dyn_clth_bck_01_l, bone2Name=dyn_parent_clth_bck_01_l, swing1LimitDegrees=5, swing2LimitDegrees=5, twistLimitDegrees=5}
 {bone1Name=dyn_clth_bck_02_l, bone2Name=dyn_clth_bck_01_l, swing1LimitDegrees=8, swing2LimitDegrees=8, twistLimitDegrees=5}
 {bone1Name=dyn_clth_bck_03_l, bone2Name=dyn_clth_bck_02_l, swing1LimitDegrees=18, swing2LimitDegrees=18, twistLimitDegrees=5}
 {bone1Name=dyn_clth_bck_04_l, bone2Name=dyn_clth_bck_03_l, swing1LimitDegrees=25, swing2LimitDegrees=25, twistLimitDegrees=5}
 {bone1Name=dyn_clth_bck_05_l, bone2Name=dyn_clth_bck_04_l, swing1LimitDegrees=30, swing2LimitDegrees=30, twistLimitDegrees=5}
 {
  bone1Name=dyn_parent_clth_bck_01_r
  bone2Name=spine_05
  swing1LimitDegrees=45
  swing1Motion=Free
  swing2LimitDegrees=45
  swing2Motion=Free
  twistLimitDegrees=45
  twistMotion=Free
 }
 {bone1Name=dyn_clth_bck_01_r, bone2Name=dyn_parent_clth_bck_01_r, swing1LimitDegrees=45, swing2LimitDegrees=45, twistLimitDegrees=5}
 {bone1Name=dyn_clth_bck_02_r, bone2Name=dyn_clth_bck_01_r, swing1LimitDegrees=15, swing2LimitDegrees=20, twistLimitDegrees=5}
 {bone1Name=dyn_clth_bck_03_r, bone2Name=dyn_clth_bck_02_r, swing1LimitDegrees=30, swing2LimitDegrees=25, twistLimitDegrees=5}
 {
  bone1Name=spine_05
  bone2Name=spine_04
  swing1LimitDegrees=45
  swing1Motion=Free
  swing2LimitDegrees=45
  swing2Motion=Free
  twistLimitDegrees=45
  twistMotion=Free
 }
 {
  bone1Name=clavicle_l
  bone2Name=spine_05
  swing1LimitDegrees=45
  swing1Motion=Free
  swing2LimitDegrees=45
  swing2Motion=Free
  twistLimitDegrees=45
  twistMotion=Free
 }
 {
  bone1Name=clavicle_r
  bone2Name=spine_05
  swing1LimitDegrees=45
  swing1Motion=Free
  swing2LimitDegrees=45
  swing2Motion=Free
  twistLimitDegrees=45
  twistMotion=Free
 }
 {bone1Name=dyn_flap_frt_02_r, bone2Name=dyn_flap_frt_01_r, swing1LimitDegrees=30, swing2LimitDegrees=30, twistLimitDegrees=5}
 {
  bone1Name=dyn_parent_flap_frt_01_l
  bone2Name=spine_05
  swing1LimitDegrees=45
  swing1Motion=Free
  swing2LimitDegrees=45
  swing2Motion=Free
  twistLimitDegrees=45
  twistMotion=Free
 }
 {bone1Name=dyn_flap_frt_01_l, bone2Name=dyn_parent_flap_frt_01_l, swing1LimitDegrees=10, swing2LimitDegrees=15, twistLimitDegrees=5}
 {bone1Name=dyn_flap_frt_02_l, bone2Name=dyn_flap_frt_01_l, swing1LimitDegrees=30, swing2LimitDegrees=30, twistLimitDegrees=5}
 {
  bone1Name=dyn_parent_flap_frt_01_r
  bone2Name=spine_05
  swing1LimitDegrees=45
  swing1Motion=Free
  swing2LimitDegrees=45
  swing2Motion=Free
  twistLimitDegrees=45
  twistMotion=Free
 }
 {bone1Name=dyn_flap_frt_01_r, bone2Name=dyn_parent_flap_frt_01_r, swing1LimitDegrees=10, swing2LimitDegrees=15, twistLimitDegrees=5}
 {
  bone1Name=spine_04
  bone2Name=spine_03
  swing1LimitDegrees=45
  swing1Motion=Free
  swing2LimitDegrees=45
  swing2Motion=Free
  twistLimitDegrees=45
  twistMotion=Free
 }
]
