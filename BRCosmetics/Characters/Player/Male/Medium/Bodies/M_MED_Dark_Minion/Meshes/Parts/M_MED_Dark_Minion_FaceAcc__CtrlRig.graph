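# Control Rig for the face accessory part of the male medium Dark Minion body, adjusting the head_offset bone that parents the dyn_ feather chains.
# Its single forward-solve step projects the head's initial pose relative to neck_02 onto the current neck_02.
# It sets head_offset halfway toward that transform using a 0.5 weight, so the current head rotation is not read.
# The hierarchy also carries flap and back-cloth chains, which the graph leaves untouched.
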ControlRigBlueprint /BRCosmetics/Characters/Player/Male/Medium/Bodies/M_MED_Dark_Minion/Meshes/Parts/M_MED_Dark_Minion_FaceAcc__CtrlRig
graphs [
 RigVMFunctionLibrary {links [], nodes []}
 RigVMModel {nodes [
   RigUnit_BeginExecution = &RigUnit_BeginExecution()
   n1 = /Script/ControlRig.RigUnit_ProjectTransformToNewParent(
    Child=(Type=Bone,Name="head")
    NewParent=(Type=Bone,Name="neck_02")
    OldParent=(Type=Bone,Name="neck_02")
    bChildInitial=True
    bOldParentInitial=True
   ) {is=struct RigVMUnitNode}
   &RigUnit_SetTransform(Item=(Type=Bone,Name="head_offset"), Weight=0.5, Value=@n1.Transform, ExecutePin=@RigUnit_BeginExecution.ExecutePin)
  ]}
]
hierarchy [
 root {loc=<0 0 0>, rot=<0 0 0 1>}
 pelvis {parent=root, loc=<0 -0.7071 93.1191>, rot=<0.0127 0.707 -0.0127 -0.707>}
 spine_01 {parent=pelvis, loc=<6.7494 -0.6388 0>, rot=<0 0 0.034 -0.9994>}
 spine_02 {parent=spine_01, loc=<7.2518 -0.035 0>, rot=<0 0 0.0238 -0.9997>}
 spine_03 {parent=spine_02, loc=<6.0104 0 0>, rot=<0 0 0.0149 -0.9999>}
 spine_04 {parent=spine_03, loc=<8.1024 0 0>, rot=<0 0 -0.0132 -0.9999>}
 spine_05 {parent=spine_04, loc=<10.0146 0 0>, rot=<0 0 -0.0015 -1>}
 clavicle_l {parent=spine_05, loc=<4.3575 -0.1466 -2.3223>, rot=<0.0465 -0.7289 -0.009 -0.6829>}
 clavicle_r {parent=spine_05, loc=<4.3575 -0.1466 2.3223>, rot=<0.7289 0.0465 -0.6829 0.009>}
 neck_01 {parent=spine_05, loc=<11.3138 0.1529 0>, rot=<0 0 -0.4266 -0.9044>}
 neck_02 {parent=neck_01, loc=<4.5627 0 0>, rot=<0 0 0.0504 -0.9987>}
 head {parent=neck_02, loc=<5.3914 0 0>, rot=<0 0 0.2282 -0.9736>}
 head_offset {parent=neck_02, loc=<5.3519 -0.1682 0>, rot=<0 0 0.2204 -0.9754>}
 dyn_fthr_topA_01 {parent=head_offset, loc=<10.2032 -7.3789 0.1695>, rot=<0.0189 0.0093 0.8381 -0.5452>}
 dyn_fthr_topA_02 {parent=dyn_fthr_topA_01, loc=<6.5345 0 0>, rot=<0 0 0.1848 -0.9828>}
 dyn_fthr_topA_03 {parent=dyn_fthr_topA_02, loc=<6.1196 0 0>, rot=<0 0 0.0327 -0.9995>}
 dyn_fthr_topB_01 {parent=head_offset, loc=<7.2759 -8.9757 -0.5778>, rot=<0.0102 -0.0933 0.9127 -0.3978>}
 dyn_fthr_topB_02 {parent=dyn_fthr_topB_01, loc=<6.5345 0 0>, rot=<0 0 0.101 -0.9949>}
 dyn_fthr_spineA_01 {parent=head_offset, loc=<3.157 -10.932 -0.0665>, rot=<0.0181 -0.0921 0.9431 -0.319>}
 dyn_fthr_spineA_02 {parent=dyn_fthr_spineA_01, loc=<6.5345 0 0>, rot=<0 0 0.101 -0.9949>}
 dyn_parent_flap_frt_01_l {parent=spine_05, loc=<5.3766 12.9255 -5.3391>, rot=<-0.0551 -0.9913 0.117 0.0254>}
 dyn_flap_frt_01_l {parent=dyn_parent_flap_frt_01_l, loc=<0 0 0>, rot=<0 0 0 -1>}
 dyn_flap_frt_02_l {parent=dyn_flap_frt_01_l, loc=<6.5345 0 0>, rot=<0 0 0.0383 -0.9993>}
 dyn_parent_flap_frt_01_r {parent=spine_05, loc=<5.4332 12.0424 7.4788>, rot=<-0.1047 -0.9873 -0.1184 -0.0131>}
 dyn_flap_frt_01_r {parent=dyn_parent_flap_frt_01_r, loc=<0 0 0>, rot=<0 0 0 1>}
 dyn_flap_frt_02_r {parent=dyn_flap_frt_01_r, loc=<6.534 0.0001 0.0001>, rot=<0 0 0.0831 -0.9965>}
 dyn_parent_clth_bck_01_l {parent=spine_05, loc=<8.9048 -8.1808 -10.2034>, rot=<-0.0397 -0.0605 0.9947 -0.0723>}
 dyn_clth_bck_01_l {parent=dyn_parent_clth_bck_01_l, loc=<0 0 0>, rot=<0.0001 0.0105 0.0093 -0.9999>}
 dyn_clth_bck_02_l {parent=dyn_clth_bck_01_l, loc=<7.0802 -0.0419 0>, rot=<0.0006 0.0176 0.043 -0.9989>}
 dyn_clth_bck_03_l {parent=dyn_clth_bck_02_l, loc=<7.0961 0 0>, rot=<0.0001 0.0368 -0.0039 -0.9993>}
 dyn_clth_bck_04_l {parent=dyn_clth_bck_03_l, loc=<7.072 0 0>, rot=<0 0 0 1>}
 dyn_clth_bck_05_l {parent=dyn_clth_bck_04_l, loc=<6.4856 0 0>, rot=<0 0 0 1>}
 dyn_parent_clth_bck_01_r {parent=spine_05, loc=<8.8772 -8.508 8.0805>, rot=<0.0037 0.0527 0.9897 -0.1332>}
 dyn_clth_bck_01_r {parent=dyn_parent_clth_bck_01_r, loc=<0 0 0>, rot=<0 0 0 1>}
 dyn_clth_bck_02_r {parent=dyn_clth_bck_01_r, loc=<7.0809 0 0>, rot=<0 0 0.0453 -0.999>}
 dyn_clth_bck_03_r {parent=dyn_clth_bck_02_r, loc=<7.0955 -0.0001 0>, rot=<0 0 0.0398 -0.9992>}
 dyn_parent_clth_bck_01 {parent=spine_04, loc=<5.8165 -11.9036 -0.0096>, rot=<0 0 0.9988 -0.0492>}
 dyn_clth_bck_01 {parent=dyn_parent_clth_bck_01, loc=<0 0 0>, rot=<0 0 0 -1>}
 dyn_clth_bck_02 {parent=dyn_clth_bck_01, loc=<7.7296 -0.0419 0>, rot=<0 0 0 -1>}
 dyn_clth_bck_03 {parent=dyn_clth_bck_02, loc=<7.5289 0 0>, rot=<0 0 0 -1>}
 dyn_clth_bck_04 {parent=dyn_clth_bck_03, loc=<7.2007 0 0>, rot=<0 0 0 -1>}
]
