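# Post-copy AnimBlueprint for the Dark Minion male head part, adding secondary motion to its feathers and cowl.
# It copies the parent mesh pose, drives head_offset from head rotation, then weights feather root bones between head, neck and torso using constraints.
# Head twist against spine_05, computed each update, blends additive feather offsets for up/down looks.
# Per-bone AnimDynamics and a cowl RigidBody follow; the Control Rig node is unconnected.
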AnimBlueprint /BRCosmetics/Characters/Player/Male/Medium/Bodies/M_MED_Dark_Minion/Meshes/Parts/Male_Medium_Dark_Minion_Head_AnimBP
target_skeleton=&/Male_Medium_Dark_Minion_Head_Skeleton
graphs [
 AnimGraph {nodes [
   "Constraint - dyn_fthr_topB_01 (PP)"!(ComponentPose, Alpha=1, ConstraintWeights_0=0.36, ConstraintWeights_1:float=0.64) -> (Pose) {
    alphaBoolBlend {blendOption=Linear}
    boneToModify {boneName=dyn_fthr_topB_01}
    constraintSetup [{targetBone {boneName=spine_05}}, {targetBone {boneName=head_offset}}]
    constraintWeights [1,1]
   }
   "Constraint - dyn_fthr_topA_01 (PP)"!(ComponentPose, Alpha=1, ConstraintWeights_0=0.2456, ConstraintWeights_1:float=0.7544) -> (Pose) {
    alphaBoolBlend {blendOption=Linear}
    boneToModify {boneName=dyn_fthr_topA_01}
    constraintSetup [{targetBone {boneName=spine_05}}, {targetBone {boneName=head_offset}}]
    constraintWeights [1,1]
   }
   Control Rig() {
    alphaBoolBlend {blendOption=Linear}
    controlRigAssetReference {blueprintRigClass=&/M_MED_Dark_Minion_FaceAcc__CtrlRig.M_MED_Dark_Minion_FaceAcc__CtrlRig_C}
   }
   n1 = Flap Node - Bone: head_offset!(
    ComponentPose:Component Space Pose Link Structure=Local To Component(LocalPose=Copy Pose From Mesh() {bUseAttachedParent=true})
    Alpha:float=1
   ) -> (Pose:Component Space Pose Link Structure) {
    is=title FortAnimGraphNode_Flap
    bUseClamp=true
    targetClampMax=0.1
    targetClampMin=-0.25
    targetComponent=RotationZ
    alphaBoolBlend {blendOption=Linear}
    sourceDrives [{
      bUseRange=false
      drivingCurve=null
      multiplier=-1
      rangeMax=1
      rangeMin=-1
      remappedMax=1
      remappedMin=0
      sourceComponent=RotationZ
      useQuaternion=true
      sourceBone {boneName=head}
     }]
    targetBone {boneName=head_offset}
   }
   n2 = "head_offset.rotateY = head.rotateY - Bone Driven Controller"!(ComponentPose:Component Space Pose Link Structure=@n1, Alpha:float=0.6) -> (Pose:Component Space Pose Link Structure) {
    is=title AnimGraphNode_BoneDrivenController
    bAffectTargetRotationY=true
    sourceComponent=RotationY
    alphaBoolBlend {blendOption=Linear}
    sourceBone {boneName=head}
    targetBone {boneName=head_offset}
   }
   n3 = "head_offset.rotateX = head.rotateX - Bone Driven Controller"!(ComponentPose:Component Space Pose Link Structure=@n2, Alpha:float=0.6) -> (Pose:Component Space Pose Link Structure) {
    is=title AnimGraphNode_BoneDrivenController
    bAffectTargetRotationX=true
    sourceComponent=RotationX
    alphaBoolBlend {blendOption=Linear}
    sourceBone {boneName=head}
    targetBone {boneName=head_offset}
   }
   n4 = "Constraint - dyn_fthr_spineA_01 (PPPPPPP)"!(
    ComponentPose=@n3
    Alpha=1
    ConstraintWeights_0=0.0208
    ConstraintWeights_1:float=0.444
    ConstraintWeights_2:float=0.017
    ConstraintWeights_3:float=0.3762
    ConstraintWeights_4:float=0.0212
    ConstraintWeights_5:float=0.0901
    ConstraintWeights_6:float=0.0305
   ) -> (Pose) {
    alphaBoolBlend {blendOption=Linear}
    boneToModify {boneName=dyn_fthr_spineA_01}
    constraintSetup [
     {targetBone {boneName=clavicle_l}}
     {targetBone {boneName=head_offset}}
     {targetBone {boneName=spine_03}}
     {targetBone {boneName=spine_05}}
     {targetBone {boneName=clavicle_r}}
     {targetBone {boneName=neck_02}}
     {targetBone {boneName=neck_01}}
    ]
    constraintWeights [1,1,1,1,1,1,1]
   }
   n5 = "Constraint - dyn_fthr_topB_01 (PP)"(ComponentPose=@n4, Alpha=1, ConstraintWeights_0=0.2054, ConstraintWeights_1=0.7946) {
    alphaBoolBlend {blendOption=Linear}
    boneToModify {boneName=dyn_fthr_topB_01}
    constraintSetup [{targetBone {boneName=spine_05}}, {targetBone {boneName=head_offset}}]
    constraintWeights [1,1]
   }
   n6 = "Constraint - dyn_fthr_topA_01 (PPPP)"!(
    ComponentPose=@n5
    Alpha=1
    ConstraintWeights_0=0.0523
    ConstraintWeights_1:float=0.0997
    ConstraintWeights_2:float=0.0456
    ConstraintWeights_3:float=0.8024
   ) -> (Pose) {
    alphaBoolBlend {blendOption=Linear}
    boneToModify {boneName=dyn_fthr_topA_01}
    constraintSetup [{targetBone {boneName=neck_02}}, {targetBone {boneName=neck_01}}, {targetBone {boneName=spine_05}}, {targetBone {boneName=head_offset}}]
    constraintWeights [1,1,1,1]
   }
   n7 = Get head_rot_up() -> (head_rot_up:double)
   n8 = "Transform (Modify) Bone - Bone: dyn_fthr_spineA_01"(ComponentPose=@n6, Alpha=@n7, Translation="0,0.1,0", Rotation="0,12,0", Scale="1,1,1") {
    rotationMode=BMM_Additive
    rotationSpace=BCS_BoneSpace
    translationMode=BMM_Additive
    translationSpace=BCS_BoneSpace
    alphaBoolBlend {blendOption=Linear}
    alphaScaleBiasClamp {bInterpResult=true, interpSpeedDecreasing=30, interpSpeedIncreasing=30}
    boneToModify {boneName=dyn_fthr_spineA_01}
    rotation {yaw=12}
    translation {y=0.1}
   }
   n9 = "Transform (Modify) Bone - Bone: dyn_fthr_topB_01"(ComponentPose=@n8, Alpha=@n7, Translation="0,0.175,0", Rotation="0,17,0", Scale="1,1,1") {
    rotationMode=BMM_Additive
    rotationSpace=BCS_BoneSpace
    translationMode=BMM_Additive
    translationSpace=BCS_BoneSpace
    alphaBoolBlend {blendOption=Linear}
    boneToModify {boneName=dyn_fthr_topB_01}
    rotation {yaw=17}
    translation {y=0.175}
   }
   n10 = "Transform (Modify) Bone - Bone: dyn_fthr_topA_01"(ComponentPose=@n9, Alpha=@n7, Translation="0,0.2,0", Rotation="0,15,0", Scale="1,1,1") {
    rotationMode=BMM_Additive
    rotationSpace=BCS_BoneSpace
    translationMode=BMM_Additive
    translationSpace=BCS_BoneSpace
    alphaBoolBlend {blendOption=Linear}
    boneToModify {boneName=dyn_fthr_topA_01}
    rotation {yaw=15}
    translation {y=0.2}
   }
   n11 = Get head_rot_dwn() -> (head_rot_dwn:double)
   n12 = "Transform (Modify) Bone - Bone: dyn_fthr_spineA_01"(ComponentPose=@n10, Alpha=@n11, Translation="0,0,0", Rotation="0,-5,0", Scale="1,1,1") {
    rotationMode=BMM_Additive
    rotationSpace=BCS_BoneSpace
    translationMode=BMM_Additive
    translationSpace=BCS_BoneSpace
    alphaBoolBlend {blendOption=Linear}
    alphaScaleBiasClamp {bInterpResult=true, interpSpeedDecreasing=30, interpSpeedIncreasing=30}
    boneToModify {boneName=dyn_fthr_spineA_01}
    rotation {yaw=-5}
   }
   n13 = "Transform (Modify) Bone - Bone: dyn_fthr_topB_01"(ComponentPose=@n12, Alpha=@n11, Translation="0,0,0", Rotation="0,-5,0", Scale="1,1,1") {
    rotationMode=BMM_Additive
    rotationSpace=BCS_BoneSpace
    translationMode=BMM_Additive
    translationSpace=BCS_BoneSpace
    alphaBoolBlend {blendOption=Linear}
    boneToModify {boneName=dyn_fthr_topB_01}
    rotation {yaw=-5}
   }
   n14 = Anim Dynamics - Bone: dyn_fthr_spin_A_01(ComponentPose=@n13, Alpha=1, ExternalForce="0,0,0") {
    angularDampingOverride=1
    angularSpringConstant=1500
    bAngularSpring=true
    bGravityOverrideInSimSpace=true
    bOverrideAngularDamping=true
    bUseGravityOverride=true
    gravityScale=0.29
    simulationSpace=RootRelative
    alphaBoolBlend {blendOption=Linear}
    alphaScaleBiasClamp {scale=0.45}
    boundBone {boneName=dyn_fthr_spin_A_01}
    componentAppliedLinearAccClamp=<1000 1000 100>
    componentLinearAccScale {x=0.3, z=0.2}
    componentLinearVelScale {x=0.3, z=0.5}
    gravityOverride {z=-150}
    physicsBodyDefinitions [{
      sphereCollisionRadius=0
      boundBone {boneName=dyn_fthr_spin_A_01}
      boxExtents=<5 10 10>
      constraintSetup {angularLimitsMax=<0 15 0>, angularLimitsMin=<0 -15 -40>, angularTarget=<0 0 0>}
      localJointOffset=<1 0 0>
     }]
   }
   n15 = Anim Dynamics - Bone: dyn_fthr_spin_B_01(ComponentPose=@n14, Alpha=1, ExternalForce="0,0,0") {
    angularDampingOverride=1
    angularSpringConstant=1500
    bAngularSpring=true
    bGravityOverrideInSimSpace=true
    bOverrideAngularDamping=true
    bUseGravityOverride=true
    gravityScale=0.29
    numSolverIterationsPreUpdate=8
    simulationSpace=RootRelative
    alphaBoolBlend {blendOption=Linear}
    alphaScaleBiasClamp {scale=0.45}
    boundBone {boneName=dyn_fthr_spin_B_01}
    componentAppliedLinearAccClamp=<1000 1000 100>
    componentLinearAccScale {x=0.3, z=0.2}
    componentLinearVelScale {x=0.3, z=0.5}
    gravityOverride {z=-150}
    physicsBodyDefinitions [{
      sphereCollisionRadius=0
      boundBone {boneName=dyn_fthr_spin_B_01}
      boxExtents=<5 10 10>
      constraintSetup {angularLimitsMax=<0 15 0>, angularLimitsMin=<0 -15 -40>, angularTarget=<0 0 0>}
      localJointOffset=<5 0 0>
     }]
   }
   n16 = Anim Dynamics - Bone: dyn_fthr_spine_A_01(ComponentPose=@n15, Alpha=1, ExternalForce="0,0,0") {
    angularDampingOverride=1
    angularSpringConstant=1500
    bAngularSpring=true
    bGravityOverrideInSimSpace=true
    bOverrideAngularDamping=true
    bUseGravityOverride=true
    gravityScale=0.29
    numSolverIterationsPreUpdate=8
    simulationSpace=RootRelative
    alphaBoolBlend {blendOption=Linear}
    alphaScaleBiasClamp {scale=0.45}
    boundBone {boneName=dyn_fthr_spine_A_01}
    componentAppliedLinearAccClamp=<1000 1000 100>
    componentLinearAccScale {x=0.3, z=0.2}
    componentLinearVelScale {x=0.3, z=0.5}
    gravityOverride {z=-150}
    physicsBodyDefinitions [{
      sphereCollisionRadius=0
      boundBone {boneName=dyn_fthr_spine_A_01}
      boxExtents=<5 10 10>
      constraintSetup {angularLimitsMax=<0 15 0>, angularLimitsMin=<0 -15 -40>, angularTarget=<0 0 0>}
      localJointOffset=<5 0 0>
     }]
   }
   n17 = Anim Dynamics - Bone: dyn_fthr_spine_B_01(ComponentPose=@n16, Alpha=1, ExternalForce="0,0,0") {
    angularDampingOverride=1
    angularSpringConstant=1500
    bAngularSpring=true
    bGravityOverrideInSimSpace=true
    bOverrideAngularDamping=true
    bUseGravityOverride=true
    gravityScale=0.29
    numSolverIterationsPreUpdate=8
    simulationSpace=RootRelative
    alphaBoolBlend {blendOption=Linear}
    alphaScaleBiasClamp {scale=0.45}
    boundBone {boneName=dyn_fthr_spine_B_01}
    componentAppliedLinearAccClamp=<1000 1000 100>
    componentLinearAccScale {x=0.3, z=0.2}
    componentLinearVelScale {x=0.3, z=0.5}
    gravityOverride {z=-150}
    physicsBodyDefinitions [{
      sphereCollisionRadius=0
      boundBone {boneName=dyn_fthr_spine_B_01}
      boxExtents=<5 10 10>
      constraintSetup {angularLimitsMax=<0 15 0>, angularLimitsMin=<0 -15 -40>, angularTarget=<0 0 0>}
      localJointOffset=<5 0 0>
     }]
   }
   n18 = Anim Dynamics - Bone: dyn_fthr_spine_B_02(ComponentPose=@n17, Alpha=1, ExternalForce="0,0,0") {
    angularDampingOverride=1
    angularSpringConstant=1500
    bAngularSpring=true
    bGravityOverrideInSimSpace=true
    bOverrideAngularDamping=true
    bUseGravityOverride=true
    gravityScale=0.29
    numSolverIterationsPreUpdate=8
    simulationSpace=RootRelative
    alphaBoolBlend {blendOption=Linear}
    alphaScaleBiasClamp {scale=0.45}
    boundBone {boneName=dyn_fthr_spine_B_02}
    componentAppliedLinearAccClamp=<1000 1000 100>
    componentLinearAccScale {x=0.3, z=0.2}
    componentLinearVelScale {x=0.3, z=0.5}
    gravityOverride {z=-150}
    physicsBodyDefinitions [{
      sphereCollisionRadius=0
      boundBone {boneName=dyn_fthr_spine_B_02}
      boxExtents=<5 10 10>
      constraintSetup {angularLimitsMax=<0 15 0>, angularLimitsMin=<0 -15 -40>, angularTarget=<0 0 0>}
      localJointOffset=<5 0 0>
     }]
   }
   n19 = Anim Dynamics - Bone: dyn_fthr_spine_C_01(ComponentPose=@n18, Alpha=1, ExternalForce="0,0,0") {
    angularDampingOverride=1
    angularSpringConstant=1500
    bAngularSpring=true
    bGravityOverrideInSimSpace=true
    bOverrideAngularDamping=true
    bUseGravityOverride=true
    gravityScale=0.29
    numSolverIterationsPreUpdate=8
    simulationSpace=RootRelative
    alphaBoolBlend {blendOption=Linear}
    alphaScaleBiasClamp {scale=0.45}
    boundBone {boneName=dyn_fthr_spine_C_01}
    componentAppliedLinearAccClamp=<1000 1000 100>
    componentLinearAccScale {x=0.3, z=0.2}
    componentLinearVelScale {x=0.3, z=0.5}
    gravityOverride {z=-150}
    physicsBodyDefinitions [{
      sphereCollisionRadius=0
      boundBone {boneName=dyn_fthr_spine_C_01}
      boxExtents=<5 10 10>
      constraintSetup {angularLimitsMax=<0 15 0>, angularLimitsMin=<0 -15 -40>, angularTarget=<0 0 0>}
      localJointOffset=<5 0 0>
     }]
   }
   n20 = Anim Dynamics - Bone: dyn_fthr_spine_C_02(ComponentPose=@n19, Alpha=1, ExternalForce="0,0,0") {
    angularDampingOverride=1
    angularSpringConstant=1500
    bAngularSpring=true
    bGravityOverrideInSimSpace=true
    bOverrideAngularDamping=true
    bUseGravityOverride=true
    gravityScale=0.29
    numSolverIterationsPreUpdate=8
    simulationSpace=RootRelative
    alphaBoolBlend {blendOption=Linear}
    alphaScaleBiasClamp {scale=0.45}
    boundBone {boneName=dyn_fthr_spine_C_02}
    componentAppliedLinearAccClamp=<1000 1000 100>
    componentLinearAccScale {x=0.3, z=0.2}
    componentLinearVelScale {x=0.3, z=0.5}
    gravityOverride {z=-150}
    physicsBodyDefinitions [{
      sphereCollisionRadius=0
      boundBone {boneName=dyn_fthr_spine_C_02}
      boxExtents=<5 10 10>
      constraintSetup {angularLimitsMax=<0 15 0>, angularLimitsMin=<0 -15 -40>, angularTarget=<0 0 0>}
      localJointOffset=<5 0 0>
     }]
   }
   n21 = Anim Dynamics - Bone: dyn_fthr_spine_D_01(ComponentPose=@n20, Alpha=1, ExternalForce="0,0,0") {
    angularDampingOverride=1
    angularSpringConstant=1500
    bAngularSpring=true
    bGravityOverrideInSimSpace=true
    bOverrideAngularDamping=true
    bUseGravityOverride=true
    gravityScale=0.29
    numSolverIterationsPreUpdate=8
    simulationSpace=RootRelative
    alphaBoolBlend {blendOption=Linear}
    alphaScaleBiasClamp {scale=0.45}
    boundBone {boneName=dyn_fthr_spine_D_01}
    componentAppliedLinearAccClamp=<1000 1000 100>
    componentLinearAccScale {x=0.3, z=0.2}
    componentLinearVelScale {x=0.3, z=0.5}
    gravityOverride {z=-150}
    physicsBodyDefinitions [{
      sphereCollisionRadius=0
      boundBone {boneName=dyn_fthr_spine_D_01}
      boxExtents=<5 10 10>
      constraintSetup {angularLimitsMax=<0 15 0>, angularLimitsMin=<0 -15 -40>, angularTarget=<0 0 0>}
      localJointOffset=<5 0 0>
     }]
   }
   n22 = Anim Dynamics - Bone: dyn_fthr_spine_E_01(ComponentPose=@n21, Alpha=1, ExternalForce="0,0,0") {
    angularDampingOverride=1
    angularSpringConstant=1500
    bAngularSpring=true
    bGravityOverrideInSimSpace=true
    bOverrideAngularDamping=true
    bUseGravityOverride=true
    gravityScale=0.29
    numSolverIterationsPreUpdate=8
    simulationSpace=RootRelative
    alphaBoolBlend {blendOption=Linear}
    alphaScaleBiasClamp {scale=0.45}
    boundBone {boneName=dyn_fthr_spine_E_01}
    componentAppliedLinearAccClamp=<1000 1000 100>
    componentLinearAccScale {x=0.3, z=0.2}
    componentLinearVelScale {x=0.3, z=0.5}
    gravityOverride {z=-150}
    physicsBodyDefinitions [{
      sphereCollisionRadius=0
      boundBone {boneName=dyn_fthr_spine_E_01}
      boxExtents=<5 10 10>
      constraintSetup {angularLimitsMax=<0 15 0>, angularLimitsMin=<0 -15 -40>, angularTarget=<0 0 0>}
      localJointOffset=<5 0 0>
     }]
   }
   n23 = Anim Dynamics - Bone: dyn_fthr_mid_A_01(ComponentPose=@n22, Alpha=1, ExternalForce="0,0,0") {
    angularDampingOverride=1
    angularSpringConstant=1500
    bAngularSpring=true
    bGravityOverrideInSimSpace=true
    bOverrideAngularDamping=true
    bUseGravityOverride=true
    gravityScale=0.25
    numSolverIterationsPreUpdate=8
    simulationSpace=RootRelative
    alphaBoolBlend {blendOption=Linear}
    alphaScaleBiasClamp {scale=0.5}
    boundBone {boneName=dyn_fthr_mid_A_01}
    componentAppliedLinearAccClamp=<1000 1000 300>
    componentLinearAccScale=<0.4 0.4 0.1>
    componentLinearVelScale=<0.4 0.4 0.1>
    gravityOverride {z=-150}
    physicsBodyDefinitions [{
      sphereCollisionRadius=0
      boundBone {boneName=dyn_fthr_mid_A_01}
      boxExtents=<5 7 7>
      constraintSetup {angularLimitsMax=<0 15 0>, angularLimitsMin=<0 -15 -25>, angularTarget=<0 0 0>}
      localJointOffset=<3 0 0>
     }]
   }
   n24 = Anim Dynamics - Bone: dyn_fthr_mid_B_01(ComponentPose=@n23, Alpha=1, ExternalForce="0,0,0") {
    angularDampingOverride=1
    angularSpringConstant=1500
    bAngularSpring=true
    bGravityOverrideInSimSpace=true
    bOverrideAngularDamping=true
    bUseGravityOverride=true
    gravityScale=0.25
    numSolverIterationsPreUpdate=8
    simulationSpace=RootRelative
    alphaBoolBlend {blendOption=Linear}
    alphaScaleBiasClamp {scale=0.5}
    boundBone {boneName=dyn_fthr_mid_B_01}
    componentAppliedLinearAccClamp=<1000 1000 300>
    componentLinearAccScale=<0.4 0.4 0.1>
    componentLinearVelScale=<0.4 0.4 0.1>
    gravityOverride {z=-150}
    physicsBodyDefinitions [{
      sphereCollisionRadius=0
      boundBone {boneName=dyn_fthr_mid_B_01}
      boxExtents=<5 7 7>
      constraintSetup {angularLimitsMax=<0 15 0>, angularLimitsMin=<0 -15 -25>, angularTarget=<0 0 0>}
      localJointOffset=<3 0 0>
     }]
   }
   n25 = Anim Dynamics - Bone: dyn_fthr_mid_B_02(ComponentPose=@n24, Alpha=1, ExternalForce="0,0,0") {
    angularDampingOverride=1
    angularSpringConstant=1500
    bAngularSpring=true
    bGravityOverrideInSimSpace=true
    bOverrideAngularDamping=true
    bUseGravityOverride=true
    gravityScale=0.25
    numSolverIterationsPreUpdate=8
    simulationSpace=RootRelative
    alphaBoolBlend {blendOption=Linear}
    alphaScaleBiasClamp {scale=0.5}
    boundBone {boneName=dyn_fthr_mid_B_02}
    componentAppliedLinearAccClamp=<1000 1000 300>
    componentLinearAccScale=<0.4 0.4 0.1>
    componentLinearVelScale=<0.4 0.4 0.1>
    gravityOverride {z=-150}
    physicsBodyDefinitions [{
      sphereCollisionRadius=0
      boundBone {boneName=dyn_fthr_mid_B_02}
      boxExtents=<5 7 7>
      constraintSetup {angularLimitsMax=<0 15 0>, angularLimitsMin=<0 -15 -25>, angularTarget=<0 0 0>}
      localJointOffset=<3 0 0>
     }]
   }
   n26 = Anim Dynamics - Bone: dyn_fthr_mid_C_01(ComponentPose=@n25, Alpha=1, ExternalForce="0,0,0") {
    angularDampingOverride=1
    angularSpringConstant=1500
    bAngularSpring=true
    bGravityOverrideInSimSpace=true
    bOverrideAngularDamping=true
    bUseGravityOverride=true
    gravityScale=0.25
    numSolverIterationsPreUpdate=8
    simulationSpace=RootRelative
    alphaBoolBlend {blendOption=Linear}
    alphaScaleBiasClamp {scale=0.5}
    boundBone {boneName=dyn_fthr_mid_C_01}
    componentAppliedLinearAccClamp=<1000 1000 300>
    componentLinearAccScale=<0.4 0.4 0.1>
    componentLinearVelScale=<0.4 0.4 0.1>
    gravityOverride {z=-150}
    physicsBodyDefinitions [{
      sphereCollisionRadius=0
      boundBone {boneName=dyn_fthr_mid_C_01}
      boxExtents=<5 7 7>
      constraintSetup {angularLimitsMax=<0 15 0>, angularLimitsMin=<0 -15 -25>, angularTarget=<0 0 0>}
      localJointOffset=<3 0 0>
     }]
   }
   n27 = Anim Dynamics - Bone: dyn_fthr_mid_C_02(ComponentPose=@n26, Alpha=1, ExternalForce="0,0,0") {
    angularDampingOverride=1
    angularSpringConstant=1500
    bAngularSpring=true
    bGravityOverrideInSimSpace=true
    bOverrideAngularDamping=true
    bUseGravityOverride=true
    gravityScale=0.25
    numSolverIterationsPreUpdate=8
    simulationSpace=RootRelative
    alphaBoolBlend {blendOption=Linear}
    alphaScaleBiasClamp {scale=0.5}
    boundBone {boneName=dyn_fthr_mid_C_02}
    componentAppliedLinearAccClamp=<1000 1000 300>
    componentLinearAccScale=<0.4 0.4 0.1>
    componentLinearVelScale=<0.4 0.4 0.1>
    gravityOverride {z=-150}
    physicsBodyDefinitions [{
      sphereCollisionRadius=0
      boundBone {boneName=dyn_fthr_mid_C_02}
      boxExtents=<5 7 7>
      constraintSetup {angularLimitsMax=<0 15 0>, angularLimitsMin=<0 -15 -25>, angularTarget=<0 0 0>}
      localJointOffset=<3 0 0>
     }]
   }
   n28 = Anim Dynamics - Bone: dyn_fthr_mid_D_01(ComponentPose=@n27, Alpha=1, ExternalForce="0,0,0") {
    angularDampingOverride=1
    angularSpringConstant=1500
    bAngularSpring=true
    bGravityOverrideInSimSpace=true
    bOverrideAngularDamping=true
    bUseGravityOverride=true
    gravityScale=0.25
    numSolverIterationsPreUpdate=8
    simulationSpace=RootRelative
    alphaBoolBlend {blendOption=Linear}
    alphaScaleBiasClamp {scale=0.5}
    boundBone {boneName=dyn_fthr_mid_D_01}
    componentAppliedLinearAccClamp=<1000 1000 300>
    componentLinearAccScale=<0.4 0.4 0.1>
    componentLinearVelScale=<0.4 0.4 0.1>
    gravityOverride {z=-150}
    physicsBodyDefinitions [{
      sphereCollisionRadius=0
      boundBone {boneName=dyn_fthr_mid_D_01}
      boxExtents=<5 7 7>
      constraintSetup {angularLimitsMax=<0 15 0>, angularLimitsMin=<0 -15 -25>, angularTarget=<0 0 0>}
      localJointOffset=<3 0 0>
     }]
   }
   n29 = Anim Dynamics - Bone: dyn_fthr_mid_E_01(ComponentPose=@n28, Alpha=1, ExternalForce="0,0,0") {
    angularDampingOverride=1
    angularSpringConstant=1500
    bAngularSpring=true
    bGravityOverrideInSimSpace=true
    bOverrideAngularDamping=true
    bUseGravityOverride=true
    gravityScale=0.25
    numSolverIterationsPreUpdate=8
    simulationSpace=RootRelative
    alphaBoolBlend {blendOption=Linear}
    alphaScaleBiasClamp {scale=0.5}
    boundBone {boneName=dyn_fthr_mid_E_01}
    componentAppliedLinearAccClamp=<1000 1000 300>
    componentLinearAccScale=<0.4 0.4 0.1>
    componentLinearVelScale=<0.4 0.4 0.1>
    gravityOverride {z=-150}
    physicsBodyDefinitions [{
      sphereCollisionRadius=0
      boundBone {boneName=dyn_fthr_mid_E_01}
      boxExtents=<5 7 7>
      constraintSetup {angularLimitsMax=<0 15 0>, angularLimitsMin=<0 -15 -25>, angularTarget=<0 0 0>}
      localJointOffset=<3 0 0>
     }]
   }
   n30 = Anim Dynamics - Bone: dyn_fthr_mid_F_01(ComponentPose=@n29, Alpha=1, ExternalForce="0,0,0") {
    angularDampingOverride=1
    angularSpringConstant=1500
    bAngularSpring=true
    bGravityOverrideInSimSpace=true
    bOverrideAngularDamping=true
    bUseGravityOverride=true
    gravityScale=0.25
    numSolverIterationsPreUpdate=8
    simulationSpace=RootRelative
    alphaBoolBlend {blendOption=Linear}
    alphaScaleBiasClamp {scale=0.5}
    boundBone {boneName=dyn_fthr_mid_F_01}
    componentAppliedLinearAccClamp=<1000 1000 300>
    componentLinearAccScale=<0.4 0.4 0.1>
    componentLinearVelScale=<0.4 0.4 0.1>
    gravityOverride {z=-150}
    physicsBodyDefinitions [{
      sphereCollisionRadius=0
      boundBone {boneName=dyn_fthr_mid_F_01}
      boxExtents=<5 7 7>
      constraintSetup {angularLimitsMax=<0 15 0>, angularLimitsMin=<0 -15 -25>, angularTarget=<0 0 0>}
      localJointOffset=<3 0 0>
     }]
   }
   n31 = Anim Dynamics - Bone: dyn_fthr_mid_G_01(ComponentPose=@n30, Alpha=1, ExternalForce="0,0,0") {
    angularDampingOverride=1
    angularSpringConstant=1500
    bAngularSpring=true
    bGravityOverrideInSimSpace=true
    bOverrideAngularDamping=true
    bUseGravityOverride=true
    gravityScale=0.25
    numSolverIterationsPreUpdate=8
    simulationSpace=RootRelative
    alphaBoolBlend {blendOption=Linear}
    alphaScaleBiasClamp {scale=0.5}
    boundBone {boneName=dyn_fthr_mid_G_01}
    componentAppliedLinearAccClamp=<1000 1000 300>
    componentLinearAccScale=<0.4 0.4 0.1>
    componentLinearVelScale=<0.4 0.4 0.1>
    gravityOverride {z=-150}
    physicsBodyDefinitions [{
      sphereCollisionRadius=0
      boundBone {boneName=dyn_fthr_mid_G_01}
      boxExtents=<5 7 7>
      constraintSetup {angularLimitsMax=<0 15 0>, angularLimitsMin=<0 -15 -25>, angularTarget=<0 0 0>}
      localJointOffset=<3 0 0>
     }]
   }
   n32 = Anim Dynamics - Bone: dyn_fthr_top_A_01(ComponentPose=@n31, Alpha=1, ExternalForce="0,0,0") {
    angularDampingOverride=1
    angularSpringConstant=2500
    bGravityOverrideInSimSpace=true
    bOverrideAngularDamping=true
    bUseGravityOverride=true
    gravityScale=0.35
    numSolverIterationsPreUpdate=8
    simulationSpace=RootRelative
    alphaBoolBlend {blendOption=Linear}
    alphaScaleBiasClamp {scale=0.55}
    boundBone {boneName=dyn_fthr_top_A_01}
    componentAppliedLinearAccClamp=<1000 1000 450>
    componentLinearAccScale=<0.4 0.4 0.2>
    componentLinearVelScale=<0.4 0.4 0.1>
    gravityOverride {z=-150}
    physicsBodyDefinitions [{
      sphereCollisionRadius=0
      boundBone {boneName=dyn_fthr_top_A_01}
      boxExtents=<5 10 10>
      constraintSetup {angularLimitsMax=<0 20 0>, angularLimitsMin=<0 -20 -40>, angularTarget=<0.5 0 0>}
      localJointOffset=<5 0 0>
     }]
   }
   n33 = Anim Dynamics - Bone: dyn_fthr_top_A_02(ComponentPose=@n32, Alpha=1, ExternalForce="0,0,0") {
    angularDampingOverride=1
    angularSpringConstant=2500
    bGravityOverrideInSimSpace=true
    bOverrideAngularDamping=true
    bUseGravityOverride=true
    gravityScale=0.35
    numSolverIterationsPreUpdate=8
    simulationSpace=RootRelative
    alphaBoolBlend {blendOption=Linear}
    alphaScaleBiasClamp {scale=0.55}
    boundBone {boneName=dyn_fthr_top_A_02}
    componentAppliedLinearAccClamp=<1000 1000 450>
    componentLinearAccScale=<0.4 0.4 0.2>
    componentLinearVelScale=<0.4 0.4 0.1>
    gravityOverride {z=-150}
    physicsBodyDefinitions [{
      sphereCollisionRadius=0
      boundBone {boneName=dyn_fthr_top_A_02}
      boxExtents=<5 10 10>
      constraintSetup {angularLimitsMax=<0 20 0>, angularLimitsMin=<0 -20 -40>, angularTarget=<1 0 0>}
      localJointOffset=<5 0 0>
     }]
   }
   n34 = Anim Dynamics - Bone: dyn_fthr_top_B_01(ComponentPose=@n33, Alpha=1, ExternalForce="0,0,0") {
    angularDampingOverride=1
    angularSpringConstant=2500
    bGravityOverrideInSimSpace=true
    bOverrideAngularDamping=true
    bUseGravityOverride=true
    gravityScale=0.35
    numSolverIterationsPreUpdate=8
    simulationSpace=RootRelative
    alphaBoolBlend {blendOption=Linear}
    alphaScaleBiasClamp {scale=0.55}
    boundBone {boneName=dyn_fthr_top_B_01}
    componentAppliedLinearAccClamp=<1000 1000 450>
    componentLinearAccScale=<0.4 0.4 0.2>
    componentLinearVelScale=<0.4 0.4 0.1>
    gravityOverride {z=-150}
    physicsBodyDefinitions [{
      sphereCollisionRadius=0
      boundBone {boneName=dyn_fthr_top_B_01}
      boxExtents=<5 10 10>
      constraintSetup {angularLimitsMax=<0 20 0>, angularLimitsMin=<0 -20 -40>, angularTarget=<1 0 0>}
      localJointOffset=<5 0 0>
     }]
   }
   n35 = Anim Dynamics - Bone: dyn_fthr_top_B_02(ComponentPose=@n34, Alpha=1, ExternalForce="0,0,0") {
    angularDampingOverride=1
    angularSpringConstant=2500
    bGravityOverrideInSimSpace=true
    bOverrideAngularDamping=true
    bUseGravityOverride=true
    gravityScale=0.35
    numSolverIterationsPreUpdate=8
    simulationSpace=RootRelative
    alphaBoolBlend {blendOption=Linear}
    alphaScaleBiasClamp {scale=0.55}
    boundBone {boneName=dyn_fthr_top_B_02}
    componentAppliedLinearAccClamp=<1000 1000 450>
    componentLinearAccScale=<0.4 0.4 0.2>
    componentLinearVelScale=<0.4 0.4 0.1>
    gravityOverride {z=-150}
    physicsBodyDefinitions [{
      sphereCollisionRadius=0
      boundBone {boneName=dyn_fthr_top_B_02}
      boxExtents=<5 10 10>
      constraintSetup {angularLimitsMax=<0 20 0>, angularLimitsMin=<0 -20 -40>, angularTarget=<1 0 0>}
      localJointOffset=<5 0 0>
     }]
   }
   n36 = Anim Dynamics - Bone: dyn_fthr_top_C_01(ComponentPose=@n35, Alpha=1, ExternalForce="0,0,0") {
    angularDampingOverride=1
    angularSpringConstant=2500
    bGravityOverrideInSimSpace=true
    bOverrideAngularDamping=true
    bUseGravityOverride=true
    gravityScale=0.35
    numSolverIterationsPreUpdate=8
    simulationSpace=RootRelative
    alphaBoolBlend {blendOption=Linear}
    alphaScaleBiasClamp {scale=0.55}
    boundBone {boneName=dyn_fthr_top_C_01}
    componentAppliedLinearAccClamp=<1000 1000 450>
    componentLinearAccScale=<0.4 0.4 0.2>
    componentLinearVelScale=<0.4 0.4 0.1>
    gravityOverride {z=-150}
    physicsBodyDefinitions [{
      sphereCollisionRadius=0
      boundBone {boneName=dyn_fthr_top_C_01}
      boxExtents=<5 5 10>
      constraintSetup {angularLimitsMax=<0 20 0>, angularLimitsMin=<0 -20 -40>, angularTarget=<1 0 0>}
      localJointOffset=<5 0 0>
     }]
   }
   n37 = Anim Dynamics - Bone: dyn_fthr_top_C_02(ComponentPose=@n36, Alpha=1, ExternalForce="0,0,0") {
    angularDampingOverride=1
    angularSpringConstant=2500
    bGravityOverrideInSimSpace=true
    bOverrideAngularDamping=true
    bUseGravityOverride=true
    gravityScale=0.35
    numSolverIterationsPreUpdate=8
    simulationSpace=RootRelative
    alphaBoolBlend {blendOption=Linear}
    alphaScaleBiasClamp {scale=0.55}
    boundBone {boneName=dyn_fthr_top_C_02}
    componentAppliedLinearAccClamp=<1000 1000 450>
    componentLinearAccScale=<0.4 0.4 0.2>
    componentLinearVelScale=<0.4 0.4 0.1>
    gravityOverride {z=-150}
    physicsBodyDefinitions [{
      sphereCollisionRadius=0
      boundBone {boneName=dyn_fthr_top_C_02}
      boxExtents=<5 10 10>
      constraintSetup {angularLimitsMax=<0 20 0>, angularLimitsMin=<0 -20 -40>, angularTarget=<1 0 0>}
      localJointOffset=<5 0 0>
     }]
   }
   n38 = Anim Dynamics - Bone: dyn_fthr_top_D_01(ComponentPose=@n37, Alpha=1, ExternalForce="0,0,0") {
    angularDampingOverride=1
    angularSpringConstant=2500
    bGravityOverrideInSimSpace=true
    bOverrideAngularDamping=true
    bUseGravityOverride=true
    gravityScale=0.35
    numSolverIterationsPreUpdate=8
    simulationSpace=RootRelative
    alphaBoolBlend {blendOption=Linear}
    alphaScaleBiasClamp {scale=0.55}
    boundBone {boneName=dyn_fthr_top_D_01}
    componentAppliedLinearAccClamp=<1000 1000 450>
    componentLinearAccScale=<0.4 0.4 0.2>
    componentLinearVelScale=<0.4 0.4 0.1>
    gravityOverride {z=-150}
    physicsBodyDefinitions [{
      sphereCollisionRadius=0
      boundBone {boneName=dyn_fthr_top_D_01}
      boxExtents=<5 10 10>
      constraintSetup {angularLimitsMax=<0 20 0>, angularLimitsMin=<0 -20 -40>, angularTarget=<1 0 0>}
      localJointOffset=<5 0 0>
     }]
   }
   n39 = Anim Dynamics - Bone: dyn_fthr_top_D_02(ComponentPose=@n38, Alpha=1, ExternalForce="0,0,0") {
    angularDampingOverride=1
    angularSpringConstant=2500
    bGravityOverrideInSimSpace=true
    bOverrideAngularDamping=true
    bUseGravityOverride=true
    gravityScale=0.35
    numSolverIterationsPreUpdate=8
    simulationSpace=RootRelative
    alphaBoolBlend {blendOption=Linear}
    alphaScaleBiasClamp {scale=0.55}
    boundBone {boneName=dyn_fthr_top_D_02}
    componentAppliedLinearAccClamp=<1000 1000 450>
    componentLinearAccScale=<0.4 0.4 0.2>
    componentLinearVelScale=<0.4 0.4 0.1>
    gravityOverride {z=-150}
    physicsBodyDefinitions [{
      sphereCollisionRadius=0
      boundBone {boneName=dyn_fthr_top_D_02}
      boxExtents=<5 10 10>
      constraintSetup {angularLimitsMax=<0 20 0>, angularLimitsMin=<0 -20 -40>, angularTarget=<1 0 0>}
      localJointOffset=<5 0 0>
     }]
   }
   n40 = Anim Dynamics - Bone: dyn_fthr_top_E_01(ComponentPose=@n39, Alpha=1, ExternalForce="0,0,0") {
    angularDampingOverride=1
    angularSpringConstant=2500
    bGravityOverrideInSimSpace=true
    bOverrideAngularDamping=true
    bUseGravityOverride=true
    gravityScale=0.35
    numSolverIterationsPreUpdate=8
    simulationSpace=RootRelative
    alphaBoolBlend {blendOption=Linear}
    alphaScaleBiasClamp {scale=0.55}
    boundBone {boneName=dyn_fthr_top_E_01}
    componentAppliedLinearAccClamp=<1000 1000 450>
    componentLinearAccScale=<0.4 0.4 0.2>
    componentLinearVelScale=<0.4 0.4 0.1>
    gravityOverride {z=-150}
    physicsBodyDefinitions [{
      sphereCollisionRadius=0
      boundBone {boneName=dyn_fthr_top_E_01}
      boxExtents=<5 10 10>
      constraintSetup {angularLimitsMax=<0 20 0>, angularLimitsMin=<0 -20 -40>, angularTarget=<1 0 0>}
      localJointOffset=<5 0 0>
     }]
   }
   n41 = Anim Dynamics - Bone: dyn_fthr_top_F_01(ComponentPose=@n40, Alpha=1, ExternalForce="0,0,0") {
    angularDampingOverride=1
    angularSpringConstant=2500
    bGravityOverrideInSimSpace=true
    bOverrideAngularDamping=true
    bUseGravityOverride=true
    gravityScale=0.35
    numSolverIterationsPreUpdate=8
    simulationSpace=RootRelative
    alphaBoolBlend {blendOption=Linear}
    alphaScaleBiasClamp {scale=0.55}
    boundBone {boneName=dyn_fthr_top_F_01}
    componentAppliedLinearAccClamp=<1000 1000 450>
    componentLinearAccScale=<0.4 0.4 0.2>
    componentLinearVelScale=<0.4 0.4 0.1>
    gravityOverride {z=-150}
    physicsBodyDefinitions [{
      sphereCollisionRadius=0
      boundBone {boneName=dyn_fthr_top_F_01}
      boxExtents=<5 10 10>
      constraintSetup {angularLimitsMax=<0 20 0>, angularLimitsMin=<0 -20 -40>, angularTarget=<1 0 0>}
      localJointOffset=<5 0 0>
     }]
   }
   n42 = RigidBody(ComponentPose=@n41) {
    bOverrideWorldGravity=true
    overridePhysicsAsset=&/M_MED_Dark_Minion_Cowl_Physics
    simulationSpace=BaseBoneSpace
    alphaBoolBlend {blendOption=Linear}
    baseBoneRef {boneName=root}
    componentAppliedLinearAccClamp=<3000 3000 3000>
    componentLinearAccScale=<0.3 0.3 0.3>
    componentLinearVelScale=<0.3 0.3 0.3>
    overrideWorldGravity {z=-980}
   }
   Output Pose(Result=Component To Local(ComponentPose=@n42)) {layerGroup=None}
  ]}
 EventGraph {nodes [
   Event BlueprintInitializeAnimation()
   > Set Skel_Mesh(Skel_Mesh=GetPartSkeletalMeshComponent())
   Event BlueprintUpdateAnimation()
   > Get!(execute:Exec) -> (then:Exec, else:Exec, Skel_Mesh:Skeletal Mesh Component Object Reference)
   > Sequence!(execute) -> (then_0, then_1, then_2:Exec)
   > Set head_rot_up(head_rot_up:double=GetAlphaFromReferenceTransformDelta(
     SkeletalMeshComponent=Get Skel_Mesh()
     BoneName=head
     BaseName=spine_05
     DeltaComponentType=QuaternionTwist
     InputRange=(X=7,Y=42)
     TwistAxis="0,0,1"
    )) -> (Output_Get:double)
   > Set head_rot_dwn(head_rot_dwn:double=GetAlphaFromReferenceTransformDelta(
     SkeletalMeshComponent=Get Skel_Mesh()
     BoneName=head
     BaseName=spine_05
     DeltaComponentType=QuaternionTwist
     InputRange=(X=1,Y=-6)
     TwistAxis="0,0,1"
    )) -> (Output_Get:double)
  ]}
]
variables [
 head_rot_dwn {type=number, default=0.001}
 head_rot_up {type=number}
 Skel_Mesh {type=SkeletalMeshComponent}
]
